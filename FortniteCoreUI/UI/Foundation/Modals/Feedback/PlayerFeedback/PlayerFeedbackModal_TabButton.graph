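# Tab button widget for the player feedback modal, derived from FortPlayerFeedbackModalTab, showing a label with an outline block and a trailing chevron.
# The current-tab update event stores bIsCurrent, recolors the label and hides the chevron on the current tab.
# Hover requests the outline's Focused state on non-current tabs, and unhover instantly disables it.
# HandleArrow collapses the chevron directly; the hover text-color nodes are left unconnected.
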WidgetBlueprint /FortniteCoreUI/UI/Foundation/Modals/Feedback/PlayerFeedback/PlayerFeedbackModal_TabButton
graphs [
 EventGraph {nodes [
   n1 = Event OnUpdatedIsCurrentTab!() -> (OutputDelegate, then, bIsCurrent:bool) {is=title K2Node_Event}
   > Set bIsCurrent(bIsCurrent:bool=Select(Option 0:bool=false, Option 1:bool=true, Index=@n1.bIsCurrent) -> (ReturnValue:bool)) -> (Output_Get:bool)
   > SetColorAndOpacity(
    self=Get TabLabel() -> (TabLabel:Common Text Object Reference)
    InColorAndOpacity=Select(
     Option 0:Slate Color Structure=Get NotActiveText() -> (NotActiveText:Slate Color Structure)
     Option 1:Slate Color Structure=Get ActiveText() -> (ActiveText:Slate Color Structure)
     Index=@n1.bIsCurrent
    ) -> (ReturnValue:Slate Color Structure)
   )
   > SetVisibility(
    self=Get Arrow() -> (Arrow:Image Object Reference)
    InVisibility=Select(Option 0:ESlateVisibility Enum=Visible, Option 1:ESlateVisibility Enum=Collapsed, Index=@n1.bIsCurrent) -> (ReturnValue:ESlateVisibility Enum)
   )
   Event On Hovered()
   > n2 = Branch(Condition=Get bIsCurrent() -> (bIsCurrent:bool))
   n3 = Get WBP_UIKit_Block_Outline() -> (WBP_UIKit_Block_Outline:WBP UIKit Block Outline Object Reference)
   RequestTransition!(execute=@n2.else, self:UIKit Block Object Reference=@n3, State:EUIKitBlockVisualState Enum=Focused) -> (then)
   Event On Unhovered()
   > RequestInstantTransition!(execute, self:UIKit Block Object Reference=@n3, State:EUIKitBlockInstantTransitionState Enum=InstantDisabled) -> (then)
   SetColorAndOpacity(
    self=Get TabLabel()
    InColorAndOpacity=Select(Option 0:Slate Color Structure=Get NotActiveText(), Option 1:Slate Color Structure=Get ActiveText(), Index=Get bIsCurrent()) -> (ReturnValue:Slate Color Structure)
   )
   SetColorAndOpacity(
    self=Get TabLabel()
    InColorAndOpacity=Select(
     Option 0=Get NotActiveTextHover() -> (NotActiveTextHover:Slate Color Structure)
     Option 1=Get ActiveTextHover() -> (ActiveTextHover:Slate Color Structure)
     Index=Get bIsCurrent()
    )
   )
   Event PreConstruct()
   Event Construct()
   Event Tick()
   Event On Selected()
   Event On Deselected()
   Reroute Node(InputPin:bool=@n1.bIsCurrent) -> (OutputPin:bool)
  ]}
 HandleArrow {nodes [
   HandleArrow() {is=title K2Node_FunctionEntry}
   > SetVisibility(self=Get Arrow(), InVisibility=Collapsed)
  ]}
]
variables [
 ActiveText {type=SlateColor, default {colorUseRule=UseColor_Specified, specifiedColor=<rgba 1 1 1 1>}}
 ActiveTextHover {type=SlateColor, default {colorUseRule=UseColor_Specified, specifiedColor=<rgba 0.0409 1 0.6172 1>}}
 bIsCurrent {type=boolean}
 NotActiveText {type=SlateColor, default {colorUseRule=UseColor_Specified, specifiedColor=<rgba 0.5765 0.5765 0.5765 1>}}
 NotActiveTextHover {type=SlateColor, default {colorUseRule=UseColor_Specified, specifiedColor=<rgba 0 0.6038 1 1>}}
]
widgets {
 info {inheritedWidgetCount=2, parentClass=/Script/FortniteUI.FortPlayerFeedbackModalTab, rootWidgetClass=&HorizontalBox, widgetCount=7}
 widgets [
  HorizontalBox_0 {widgetClassPath=&HorizontalBox, widget=HorizontalBox_0}
  ContentBorder {
   widgetClassPath=&Border
   bIsVariable=true
   parent=HorizontalBox_0
   widget=ContentBorder
   properties {
    horizontalAlignment=HAlign_Center
    verticalAlignment=VAlign_Center
    background {
     bIsDynamicallyLoaded=false
     drawAs=NoDrawType
     imageType=NoImage
     mirroring=NoMirror
     resourceName=None
     resourceObject=null
     tiling=NoTile
     imageSize=<32 32>
     margin=<ltrb 0 0 0 0>
     outlineSettings {
      bUseBrushTransparency=false
      roundingType=HalfHeightRadius
      width=0
      color {colorUseRule=UseColor_Specified, specifiedColor=<rgba 0 0 0 0>}
      cornerRadii=<0 0 0 0>
     }
     tintColor {colorUseRule=UseColor_Specified, specifiedColor=<rgba 1 1 1 1>}
     uVRegion {bIsValid=false, max=<0 0>, min=<0 0>}
    }
    padding=<ltrb 0 0 0 0>
   }
   slot {class=HorizontalBoxSlot, properties {verticalAlignment=VAlign_Center}}
  }
  Overlay_0 {
   widgetClassPath=&Overlay
   parent=ContentBorder
   widget=Overlay_0
   slot {class=BorderSlot, properties {horizontalAlignment=HAlign_Center, verticalAlignment=VAlign_Center, padding=<ltrb 0 0 0 0>}}
  }
  TabLabel {
   widgetClassPath=&CommonTextBlock
   bInherited=true
   bIsVariable=true
   parent=Overlay_0
   widget=TabLabel
   properties {
    justification=Center
    mobileFontSizeMultiplier=1.3
    scrollStyle=&TextScrollStyle-Base_C
    text="{Text}"
    font {
     bForceMonospaced=false
     bMaterialIsStencil=false
     fontMaterial=null
     fontObject=&HeadingNow
     letterSpacing=0
     monospacedWidth=1
     size=18
     skewAmount=0
     typefaceFontName=75 Medium
     outlineSettings {
      bApplyOutlineToDropShadows=false
      bMiteredCorners=false
      bSeparateFillAlpha=false
      outlineMaterial=null
      outlineSize=0
      outlineColor=<rgba 0 0 0 1>
     }
    }
    shadowOffset=<0 0>
   }
   slot {class=OverlaySlot}
  }
  WBP_UIKit_Block_Outline {
   widgetClassPath=/FortUILibrary/UIKit/Buttons/Blocks/WBP_UIKit_Block_Outline
   bIsVariable=true
   parent=Overlay_0
   widget=WBP_UIKit_Block_Outline
   slot {class=OverlaySlot, properties {horizontalAlignment=HAlign_Fill, verticalAlignment=VAlign_Fill}}
  }
  Arrow {
   widgetClassPath=&Image
   bIsVariable=true
   parent=HorizontalBox_0
   widget=Arrow
   properties {brush {
     bIsDynamicallyLoaded=false
     drawAs=Image
     imageType=NoImage
     mirroring=NoMirror
     resourceName=None
     resourceObject=/FortUILibrary/Textures/MaterialInstances/MI_UIKit_Icon_Chevron
     tiling=NoTile
     imageSize=<20 20>
     margin=<ltrb 0 0 0 0>
     outlineSettings {
      bUseBrushTransparency=false
      roundingType=HalfHeightRadius
      width=0
      color {colorUseRule=UseColor_Specified, specifiedColor=<rgba 0 0 0 0>}
      cornerRadii=<0 0 0 0>
     }
     tintColor {colorUseRule=UseColor_Specified, specifiedColor=<rgba 1 1 1 1>}
     uVRegion {bIsValid=false, max=<0 0>, min=<0 0>}
    }}
   slot {class=HorizontalBoxSlot, properties {horizontalAlignment=HAlign_Center, verticalAlignment=VAlign_Center, padding=<ltrb 12 0 12 0>}}
  }
  InputActionWidget {widgetClassPath=&CommonActionWidget, bInherited=true, widget=null}
 ]
}
