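# Player feedback modal widget on FortPlayerFeedbackModal: a UIKit dialog with title, disclaimer, subscreen tabs, scrolling subscreen container and Close, Back and Next buttons.
# Construct sets the container as scroll widget and runs SetMobileStyle, which in compact UI hides Close, pads the title for touch and sets dialog padding by orientation.
# Input-mode changes switch scroll navigation between centered for gamepad and into-view otherwise.
# Hardware visibility borders swap Close for a top-right button on touch.
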WidgetBlueprint /FortniteCoreUI/UI/Foundation/Modals/Feedback/PlayerFeedback/PlayerFeedbackModal
graphs [
 EventGraph {nodes [
   Event Construct()
   > SetScrollWidget!(execute, ToScroll:Widget Object Reference=Get SubscreenContainer() -> (SubscreenContainer:Scroll Box Object Reference)) -> (then)
   > SetMobileStyle() {is=title K2Node_CallFunction}
   n1 = Event OnInputModeChanged!() -> (OutputDelegate, then, bUsingGamepad:bool) {is=title K2Node_Event}
   > n2 = Branch(Condition=@n1.bUsingGamepad)
   n3 = Get SubscreenContainer()
   > Set NavigationDestination!(execute, NavigationDestination:EDescendantScrollDestination Enum=Center, self:Scroll Box Object Reference=@n3) -> (then, Output_Get:EDescendantScrollDestination Enum=IntoView)
   Set NavigationDestination(execute=@n2.else, NavigationDestination=IntoView, self=@n3) -> (Output_Get=IntoView)
  ]}
 SetMobileStyle {nodes [
   SetMobileStyle() {is=title K2Node_FunctionEntry}
   > Branch(Condition=IsUsingCompactUI())
   > Set Visibility(Visibility=Collapsed, self=Get Button_Close() -> (Button_Close:Common Button Base Object Reference))
   > n1 = Sequence()
   > Branch(Condition=IsUsingInputType!(InputType:ECommonInputType Enum=Touch, PlayerController:Player Controller Object Reference=GetOwningPlayer()) -> (ReturnValue:bool))
   n2 = SlotAsVerticalBoxSlot(Widget=Get Text_ModalTitle() -> (Text_ModalTitle:Common Text Object Reference))
   n3 = Get!(self:Vertical Box Slot Object Reference=@n2) -> (Padding_Left:float=~, Padding_Top:float=~, Padding_Right:float=~, Padding_Bottom:float=~)
   > SetPadding!(
    execute
    self:Vertical Box Slot Object Reference=@n2
    InPadding_Left:float=@n3
    InPadding_Top:float=@n3.Padding_Top
    InPadding_Right:float=96
    InPadding_Bottom:float=@n3.Padding_Bottom
   ) -> (then)
   n4 = Get WBP_UIKit_Dialog_Base() -> (WBP_UIKit_Dialog_Base:WBP UIKit Dialog Internal Object Reference)
   n5 = Get BigDialogPadding() -> (BigDialogPadding:float)
   n6 = Get SmallDialogPadding() -> (SmallDialogPadding:float)
   n7 = Get BigDialogPadding()
   n8 = Get SmallDialogPadding()
   n9 = Select!(
    Option 0_Left:float=@n7
    Option 0_Top:float=@n6
    Option 0_Right:float=@n7
    Option 0_Bottom:float=@n6
    Option 1_Left:float=@n8
    Option 1_Top:float=@n5
    Option 1_Right:float=@n8
    Option 1_Bottom:float=@n5
    Index=IsPortrait!(self:Fort Display System VM Object Reference=Get FortDisplaySystemVM() -> (FortDisplaySystemVM:Fort Display System VM Object Reference)) -> (ReturnValue:bool)
   ) -> (ReturnValue:Margin Structure)
   SetPlacement!(
    execute=@n1.then_1
    self:WBP UIKit Dialog Internal Object Reference=@n4
    Widget:Widget Object Reference=Get SizeBox_Dialog!(self:WBP UIKit Dialog Internal Object Reference=@n4) -> (SizeBox_Dialog:Size Box Object Reference)
    Placement_HorizontalAlignment_4_#1:EHorizontalAlignment Enum=HAlign_Fill
    Placement_VerticalAlignment_5_#2:EVerticalAlignment Enum=VAlign_Fill
    Placement_Padding_8_#3:Margin Structure=@n9
    Placement_Grid_20_#4:S UI Grid Cell Structure=(Row_15_#1=0,RowSpan_17_#5=0,Column_19_#6=0,ColumnSpan_21_#7=0,Layer_23_#8=0,Nudge_26_#9=(X=0,Y=0))
   ) -> (then)
   Set!(
    execute
    DialogPlacement_HorizontalAlignment_4_#1:EHorizontalAlignment Enum=HAlign_Fill
    DialogPlacement_VerticalAlignment_5_#2:EVerticalAlignment Enum=VAlign_Fill
    DialogPlacement_Padding_8_#3:Margin Structure=@n9
    DialogPlacement_Grid_20_#4:S UI Grid Cell Structure=(Row_15_#1=0,RowSpan_17_#5=0,Column_19_#6=0,ColumnSpan_21_#7=0,Layer_23_#8=0,Nudge_26_#9=(X=0,Y=0))
    self:WBP UIKit Dialog Internal Object Reference=Get WBP_UIKit_Dialog_Base()
   ) -> (then, Output_Get:S UI Placement Structure)
  ]}
]
variables [Mobile Scale {type=number, default=1.4}]
widgets {
 info {
  inheritedWidgetCount=9
  namedSlotCount=1
  parentClass=/Script/FortniteUI.FortPlayerFeedbackModal
  rootWidgetClass=&Overlay
  widgetCount=16
 }
 widgets [
  RootOverlay {widgetClassPath=&Overlay, bIsVariable=true, widget=RootOverlay}
  WBP_UIKit_Dialog_Base {
   widgetClassPath=/FortUILibrary/Systems/Dialog/WBP_UIKit_Dialog_Internal
   bIsVariable=true
   parent=RootOverlay
   widget=WBP_UIKit_Dialog_Base
   properties {
    dialogPlacement {
     horizontal Alignment=HAlign_Fill
     vertical Alignment=VAlign_Fill
     grid {column=0, column Span=0, layer=1, row=0, row Span=0, nudge=<0 0>}
     padding=<ltrb 300 80 300 80>
    }
    dialogSIze {
     height Override {enable=false, value=800}
     max Aspect Ratio {enable=false, value=0}
     max Desired Height {enable=true, value=1800}
     max Desired Width {enable=true, value=2000}
     min Aspect Ratio {enable=false, value=0}
     min Desired Height {enable=false, value=0}
     min Desired Width {enable=false, value=0}
     width Override {enable=false, value=1200}
    }
   }
   slot {class=OverlaySlot, properties {horizontalAlignment=HAlign_Fill, verticalAlignment=VAlign_Fill}}
  }
  Overlay_0 {widgetClassPath=&Overlay, namedSlotHost=WBP_UIKit_Dialog_Base, widget=Overlay_0}
  VerticalBox_4 {
   widgetClassPath=&VerticalBox
   parent=Overlay_0
   widget=VerticalBox_4
   slot {class=OverlaySlot, properties {horizontalAlignment=HAlign_Fill, verticalAlignment=VAlign_Fill, padding=<ltrb 80 80 80 48>}}
  }
  Text_ModalTitle {
   widgetClassPath=&CommonTextBlock
   bInherited=true
   bIsVariable=true
   parent=VerticalBox_4
   widget=Text_ModalTitle
   properties {
    scrollStyle=&TextScrollStyle-Base_C
    text=Report Feedback
    textTransformPolicy=ToUpper
    font {
     bForceMonospaced=false
     bMaterialIsStencil=false
     fontMaterial=null
     fontObject=&HeadingNow
     letterSpacing=0
     monospacedWidth=1
     size=25.5
     skewAmount=0
     typefaceFontName=86 Bold Italic
     outlineSettings {
      bApplyOutlineToDropShadows=false
      bMiteredCorners=false
      bSeparateFillAlpha=false
      outlineMaterial=null
      outlineSize=0
      outlineColor=<rgba 0 0 0 1>
     }
    }
    shadowOffset=<0 0>
   }
   slot {class=VerticalBoxSlot, properties {verticalAlignment=VAlign_Top, padding=<ltrb 0 0 0 32>}}
  }
  Text_ModalDisclaimer {
   widgetClassPath=&CommonTextBlock
   bInherited=true
   bIsVariable=true
   parent=VerticalBox_4
   widget=Text_ModalDisclaimer
   properties {
    autoWrapText=true
    text=Feedback modal disclaimer
    wrapTextAt=1000
    font {
     bForceMonospaced=false
     bMaterialIsStencil=false
     fontMaterial=null
     fontObject=&HeadingNow
     letterSpacing=0
     monospacedWidth=1
     size=19.5
     skewAmount=0
     typefaceFontName=74 Regular
     outlineSettings {
      bApplyOutlineToDropShadows=false
      bMiteredCorners=false
      bSeparateFillAlpha=false
      outlineMaterial=null
      outlineSize=0
      outlineColor=<rgba 0 0 0 1>
     }
    }
    shadowOffset=<0 0>
   }
   slot {class=VerticalBoxSlot, properties {padding=<ltrb 0 -20 0 32>}}
  }
  SubscreenTabs {
   widgetClassPath=&DynamicEntryBox
   bInherited=true
   bIsVariable=true
   parent=VerticalBox_4
   widget=SubscreenTabs
   properties {
    entryBoxType=Wrap
    entryWidgetClass=&/PlayerFeedbackModal_TabButton.PlayerFeedbackModal_TabButton_C
    visibility=Visible
    entrySpacing=<0 10>
   }
   slot {class=VerticalBoxSlot}
  }
  SubscreenContainer {
   widgetClassPath=&ScrollBox
   bInherited=true
   bIsVariable=true
   parent=VerticalBox_4
   widget=SubscreenContainer
   properties {
    bIsVolatile=true
    consumeMouseWheel=Always
    navigation=&self:WidgetTree.SubscreenContainer.WidgetNavigation_0
    scrollWhenFocusChanges=AnimatedScroll
   }
   slot {class=VerticalBoxSlot, properties {padding=<ltrb -16 32 -16 32>, size {sizeRule=Fill, value=1}}}
  }
  HorizontalBox {
   widgetClassPath=&HorizontalBox
   parent=VerticalBox_4
   widget=HorizontalBox
   slot {class=VerticalBoxSlot, properties {horizontalAlignment=HAlign_Left}}
  }
  VisibilityBorder_NotTouch {
   widgetClassPath=&CommonHardwareVisibilityBorder
   parent=HorizontalBox
   widget=VisibilityBorder_NotTouch
   properties {
    horizontalAlignment=HAlign_Left
    verticalAlignment=VAlign_Center
    padding=<ltrb 0 0 24 0>
    visibilityQuery {
     autoDescription=" NONE( Input.Touch )"
     tokenStreamVersion=0
     userDescription=""
     queryTokenStream [0,1,3,1,0]
     tagDictionary [{tagName=Input.Touch}]
    }
   }
   slot {class=HorizontalBoxSlot}
  }
  Button_Close {
   widgetClassPath=/FortUILibrary/UIKit/Buttons/WBP_UIKit_Button_Loud
   bInherited=true
   bIsVariable=true
   parent=VisibilityBorder_NotTouch
   widget=Button_Close
   properties {
    bHideInputActionWithKeyboard=true
    text=Close
    visibility=Visible
    triggeringInputAction {dataTable=&InputActionDataTable, rowName=Close}
   }
   slot {class=BorderSlot, properties {horizontalAlignment=HAlign_Left, verticalAlignment=VAlign_Center, padding=<ltrb 0 0 24 0>}}
  }
  Button_Back {
   widgetClassPath=&WBP_UIKit_Button_Regular
   bInherited=true
   bIsVariable=true
   parent=HorizontalBox
   widget=Button_Back
   properties {
    bHideInputActionWithKeyboard=true
    buttonType=Quiet
    mobileMinHeight=88
    text=Back
    triggeringInputAction {dataTable=&InputActionDataTable, rowName=GenericBack}
   }
   slot {class=HorizontalBoxSlot, properties {horizontalAlignment=HAlign_Left, verticalAlignment=VAlign_Center, padding=<ltrb 0 0 24 0>}}
  }
  Button_Next {
   widgetClassPath=&WBP_UIKit_Button_Regular
   bInherited=true
   bIsVariable=true
   parent=HorizontalBox
   widget=Button_Next
   properties {
    bHideInputActionWithKeyboard=true
    buttonType=Quiet
    mobileMinHeight=88
    text=Next
    triggeringInputAction {dataTable=&InputActionDataTable, rowName=GenericForward}
   }
   slot {class=HorizontalBoxSlot, properties {horizontalAlignment=HAlign_Left, verticalAlignment=VAlign_Center}}
  }
  VisiblityBorder_Touch {
   widgetClassPath=&CommonHardwareVisibilityBorder
   parent=Overlay_0
   widget=VisiblityBorder_Touch
   properties {
    horizontalAlignment=HAlign_Right
    verticalAlignment=VAlign_Top
    padding=<ltrb 0 24 24 0>
    visibilityQuery {
     autoDescription=" ANY( Input.Touch )"
     tokenStreamVersion=0
     userDescription=""
     queryTokenStream [0,1,1,1,0]
     tagDictionary [{tagName=Input.Touch}]
    }
   }
   slot {class=OverlaySlot, properties {horizontalAlignment=HAlign_Right}}
  }
  Button_CloseTouch {
   widgetClassPath=/FortUILibrary/Foundation/Buttons/CloseButton
   bInherited=true
   bIsVariable=true
   parent=VisiblityBorder_Touch
   widget=Button_CloseTouch
   properties {
    onButtonBaseClicked=(&self:WidgetTree.Button_CloseTouch.HandleOnButtonClicked)
    onButtonBaseDoubleClicked=(&self:WidgetTree.Button_CloseTouch.HandleOnButtonDoubleClicked)
    onButtonBaseHovered=(&self:WidgetTree.Button_CloseTouch.HandleOnButtonHovered)
    onButtonBaseUnhovered=(&self:WidgetTree.Button_CloseTouch.HandleOnButtonUnhovered)
    onSelectedChangedBase=(&self:WidgetTree.Button_CloseTouch.HandleOnSelectedChanged)
   }
   slot {class=BorderSlot, properties {horizontalAlignment=HAlign_Right, verticalAlignment=VAlign_Top, padding=<ltrb 0 24 24 0>}}
  }
  CloseButton {widgetClassPath=&CommonButtonBase, bInherited=true, widget=null}
 ]
}
