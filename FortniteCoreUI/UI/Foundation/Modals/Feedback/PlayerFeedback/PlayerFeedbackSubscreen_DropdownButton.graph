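# Category dropdown button widget for the player feedback modal, derived from FortPlayerFeedbackModalCategoryButton, with backplates, label and chevron.
# When the dropdown status updates, it rotates the chevron between 0 and 90 degrees and toggles the selected backplate.
# A transparent click-through button drives the UIKit outline block, transitioning to focused on hover and instantly disabling it on unhover.
# Switching input to gamepad only prints a debug string, and the collapsed/expanded text color variables are unused in the graph.
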WidgetBlueprint /FortniteCoreUI/UI/Foundation/Modals/Feedback/PlayerFeedback/PlayerFeedbackSubscreen_DropdownButton
graphs [EventGraph {nodes [
   "On Hovered (Button_Outline)"() {is=title K2Node_ComponentBoundEvent}
   n1 = Get WBP_UIKit_Block_Outline() -> (WBP_UIKit_Block_Outline:WBP UIKit Block Outline Object Reference)
   > RequestTransition!(execute, self:UIKit Block Object Reference=@n1, State:EUIKitBlockVisualState Enum=Focused) -> (then)
   "On Unhovered (Button_Outline)"() {is=title K2Node_ComponentBoundEvent}
   > RequestInstantTransition!(execute, self:UIKit Block Object Reference=@n1, State:EUIKitBlockInstantTransitionState Enum=InstantDisabled) -> (then)
   n2 = Event On Input Method Changed!() -> (OutputDelegate, then, CurrentInputType:ECommonInputType Enum) {is=title K2Node_Event}
   > Branch(Condition="Equal (Enum)"(A=@n2.CurrentInputType, B=Gamepad))
   > PrintString!(execute, InString=Gamepad, bPrintToScreen, bPrintToLog, TextColor, Duration, Key) -> (then)
   n3 = Event OnDropdownMenuStatusUpdated!() -> (OutputDelegate, then, bIsExpanded:bool) {is=title K2Node_Event}
   > SetRenderTransformAngle!(
    execute
    self:Widget Object Reference=Get Arrow() -> (Arrow:Image Object Reference)
    Angle:float=SelectFloat(A=90, B=0, bPickA=@n3.bIsExpanded)
   ) -> (then)
   > SetVisibility(
    self=Get SelectedBackplate() -> (SelectedBackplate:Image Object Reference)
    InVisibility="To Visibility (Boolean)"!(bIsVisible:bool=@n3.bIsExpanded, TrueVisibility:ESlateVisibility Enum=Visible, FalseVisibility:ESlateVisibility Enum=Collapsed) -> (ReturnValue:ESlateVisibility Enum=Visible)
   )
  ]}]
variables [
 CollapsedTextColor {type=SlateColor, default {colorUseRule=UseColor_Specified, specifiedColor=<rgba 0.3325 0.7157 1 1>}}
 ExpandedTextColor {type=SlateColor, default {colorUseRule=UseColor_Specified, specifiedColor=<rgba 0.0021 0.013 0.3968 1>}}
 In Font Info Font Family {type=Object, default=null}
]
widgets {
 info {
  inheritedWidgetCount=3
  parentClass=/Script/FortniteUI.FortPlayerFeedbackModalCategoryButton
  rootWidgetClass=&Overlay
  widgetCount=10
 }
 widgets [
  Overlay_0 {widgetClassPath=&Overlay, widget=Overlay_0}
  Spacer_101 {
   widgetClassPath=&Spacer
   parent=Overlay_0
   widget=Spacer_101
   properties {size=<800 64>}
   slot {class=OverlaySlot, properties {horizontalAlignment=HAlign_Fill}}
  }
  SelectedBackplate {
   widgetClassPath=&Image
   bIsVariable=true
   parent=Overlay_0
   widget=SelectedBackplate
   properties {
    visibility=Collapsed
    brush {
     bIsDynamicallyLoaded=false
     drawAs=Image
     imageType=NoImage
     mirroring=NoMirror
     resourceName=None
     resourceObject=/FortUILibrary/UIKit/Buttons/MaterialInstances/CTAButtons/MI_UIKit_BgRegular
     tiling=NoTile
     imageSize=<32 32>
     margin=<ltrb 0 0 0 0>
     outlineSettings {
      bUseBrushTransparency=false
      roundingType=HalfHeightRadius
      width=0
      color {colorUseRule=UseColor_Specified, specifiedColor=<rgba 0 0 0 0>}
      cornerRadii=<0 0 0 0>
     }
     tintColor {colorUseRule=UseColor_Specified, specifiedColor=<rgba 1 1 1 1>}
     uVRegion {bIsValid=false, max=<0 0>, min=<0 0>}
    }
   }
   slot {class=OverlaySlot, properties {horizontalAlignment=HAlign_Fill, verticalAlignment=VAlign_Fill}}
  }
  Backplate {
   widgetClassPath=&Image
   bIsVariable=true
   parent=Overlay_0
   widget=Backplate
   properties {
    visibility=SelfHitTestInvisible
    brush {
     bIsDynamicallyLoaded=false
     drawAs=Image
     imageType=NoImage
     mirroring=NoMirror
     resourceName=None
     resourceObject=&MI_UIKit_BgRegular
     tiling=NoTile
     imageSize=<32 32>
     margin=<ltrb 0 0 0 0>
     outlineSettings {
      bUseBrushTransparency=false
      roundingType=HalfHeightRadius
      width=0
      color {colorUseRule=UseColor_Specified, specifiedColor=<rgba 0 0 0 0>}
      cornerRadii=<0 0 0 0>
     }
     tintColor {colorUseRule=UseColor_Specified, specifiedColor=<rgba 1 1 1 1>}
     uVRegion {bIsValid=false, max=<0 0>, min=<0 0>}
    }
   }
   slot {class=OverlaySlot, properties {horizontalAlignment=HAlign_Fill, verticalAlignment=VAlign_Fill}}
  }
  Text_Label {
   widgetClassPath=&CommonTextBlock
   bInherited=true
   bIsVariable=true
   parent=Overlay_0
   widget=Text_Label
   properties {
    applyLineHeightToBottomLine=false
    autoWrapText=true
    bIsScrollingEnabled=false
    mobileFontSizeMultiplier=1.3
    text=Epic QA
    colorAndOpacity {colorUseRule=UseColor_Specified, specifiedColor=<rgba 0 0 0 1>}
    font {
     bForceMonospaced=false
     bMaterialIsStencil=false
     fontMaterial=null
     fontObject=&HeadingNow
     letterSpacing=0
     monospacedWidth=1
     size=18
     skewAmount=0
     typefaceFontName=75 Medium
     outlineSettings {
      bApplyOutlineToDropShadows=false
      bMiteredCorners=false
      bSeparateFillAlpha=false
      outlineMaterial=null
      outlineSize=0
      outlineColor=<rgba 0 0 0 1>
     }
    }
    shadowOffset=<0 0>
   }
   slot {class=OverlaySlot, properties {horizontalAlignment=HAlign_Fill, verticalAlignment=VAlign_Center, padding=<ltrb 32 12 72 12>}}
  }
  DropdownWidget {
   widgetClassPath=&Overlay
   bInherited=true
   bIsVariable=true
   parent=Overlay_0
   widget=DropdownWidget
   slot {class=OverlaySlot, properties {horizontalAlignment=HAlign_Right, verticalAlignment=VAlign_Center, padding=<ltrb 0 0 24 0>}}
  }
  Arrow {
   widgetClassPath=&Image
   bIsVariable=true
   parent=DropdownWidget
   widget=Arrow
   properties {
    bFlipForRightToLeftFlowDirection=true
    visibility=SelfHitTestInvisible
    brush {
     bIsDynamicallyLoaded=false
     drawAs=Image
     imageType=NoImage
     mirroring=NoMirror
     resourceName=None
     resourceObject=/FortUILibrary/Textures/MaterialInstances/MI_UIKit_Icon_Chevron
     tiling=NoTile
     imageSize=<40 40>
     margin=<ltrb 0 0 0 0>
     outlineSettings {
      bUseBrushTransparency=false
      roundingType=HalfHeightRadius
      width=0
      color {colorUseRule=UseColor_Specified, specifiedColor=<rgba 0 0 0 0>}
      cornerRadii=<0 0 0 0>
     }
     tintColor {colorUseRule=UseColor_Specified, specifiedColor=<rgba 1 1 1 1>}
     uVRegion {bIsValid=false, max=<0 0>, min=<0 0>}
    }
    colorAndOpacity=<rgba 0 0 0 1>
    renderTransform {angle=90, scale=<1 1>, shear=<0 0>, translation=<0 0>}
   }
   slot {class=OverlaySlot}
  }
  Button_Outline {
   widgetClassPath=/Script/FortniteUI.FortClickThroughButton
   bIsVariable=true
   parent=Overlay_0
   widget=Button_Outline
   properties {
    isFocusable=false
    widgetStyle {
     clickedSlateSound {resourceObject=null}
     disabled {
      bIsDynamicallyLoaded=false
      drawAs=NoDrawType
      imageType=NoImage
      mirroring=NoMirror
      resourceName=None
      resourceObject=null
      tiling=NoTile
      imageSize=<0 0>
      margin=<ltrb 0 0 0 0>
      outlineSettings {
       bUseBrushTransparency=false
       roundingType=HalfHeightRadius
       width=0
       color {colorUseRule=UseColor_Specified, specifiedColor=<rgba 0 0 0 0>}
       cornerRadii=<0 0 0 0>
      }
      tintColor {colorUseRule=UseColor_Specified, specifiedColor=<rgba 1 1 1 1>}
      uVRegion {bIsValid=false, max=<0 0>, min=<0 0>}
     }
     disabledForeground {colorUseRule=UseColor_Specified, specifiedColor=<rgba 0.5271 0.5271 0.5271 1>}
     hovered {
      bIsDynamicallyLoaded=false
      drawAs=NoDrawType
      imageType=NoImage
      mirroring=NoMirror
      resourceName=None
      resourceObject=null
      tiling=NoTile
      imageSize=<32 32>
      margin=<ltrb 0 0 0 0>
      outlineSettings {
       bUseBrushTransparency=true
       roundingType=FixedRadius
       width=1
       color {colorUseRule=UseColor_Specified, specifiedColor=<rgba 0.7243 0.7243 0.7243 1>}
       cornerRadii=<4 4 4 4>
      }
      tintColor {colorUseRule=UseColor_Specified, specifiedColor=<rgba 0.7243 0.7243 0.7243 1>}
      uVRegion {bIsValid=false, max=<0 0>, min=<0 0>}
     }
     hoveredForeground {colorUseRule=UseColor_Specified, specifiedColor=<rgba 0 1 1 1>}
     hoveredSlateSound {resourceObject=null}
     normal {
      bIsDynamicallyLoaded=false
      drawAs=NoDrawType
      imageType=NoImage
      mirroring=NoMirror
      resourceName=None
      resourceObject=null
      tiling=NoTile
      imageSize=<32 32>
      margin=<ltrb 0 0 0 0>
      outlineSettings {
       bUseBrushTransparency=true
       roundingType=FixedRadius
       width=1
       color {colorUseRule=UseColor_Specified, specifiedColor=<rgba 0.6951 0.6951 0.6951 1>}
       cornerRadii=<4 4 4 4>
      }
      tintColor {colorUseRule=UseColor_Specified, specifiedColor=<rgba 0.4955 0.4955 0.4955 1>}
      uVRegion {bIsValid=false, max=<0 0>, min=<0 0>}
     }
     normalForeground {colorUseRule=UseColor_Specified, specifiedColor=<rgba 0 1 1 1>}
     normalPadding=<ltrb 0 0 0 0>
     pressed {
      bIsDynamicallyLoaded=false
      drawAs=NoDrawType
      imageType=NoImage
      mirroring=NoMirror
      resourceName=None
      resourceObject=null
      tiling=NoTile
      imageSize=<32 32>
      margin=<ltrb 0 0 0 0>
      outlineSettings {
       bUseBrushTransparency=true
       roundingType=FixedRadius
       width=1
       color {colorUseRule=UseColor_Specified, specifiedColor=<rgba 0.7243 0.7243 0.7243 1>}
       cornerRadii=<4 4 4 4>
      }
      tintColor {colorUseRule=UseColor_Specified, specifiedColor=<rgba 0.3843 0.3843 0.3843 1>}
      uVRegion {bIsValid=false, max=<0 0>, min=<0 0>}
     }
     pressedForeground {colorUseRule=UseColor_Specified, specifiedColor=<rgba 0 1 1 1>}
     pressedPadding=<ltrb 0 0 0 0>
     pressedSlateSound {resourceObject=null}
    }
   }
   slot {class=OverlaySlot, properties {horizontalAlignment=HAlign_Fill, verticalAlignment=VAlign_Fill}}
  }
  WBP_UIKit_Block_Outline {
   widgetClassPath=/FortUILibrary/UIKit/Buttons/Blocks/WBP_UIKit_Block_Outline
   bIsVariable=true
   parent=Button_Outline
   widget=WBP_UIKit_Block_Outline
   slot {class=ButtonSlot, properties {horizontalAlignment=HAlign_Fill, verticalAlignment=VAlign_Fill, padding=<ltrb 0 0 0 0>}}
  }
  InputActionWidget {
   widgetClassPath=&CommonActionWidget
   bInherited=true
   bIsVariable=true
   parent=Overlay_0
   widget=InputActionWidget
   properties {
    progressMaterialParam=Percentage
    visibility=Collapsed
    progressMaterialBrush {
     bIsDynamicallyLoaded=false
     drawAs=Image
     imageType=NoImage
     mirroring=NoMirror
     resourceName=None
     resourceObject=/Game/UI/Materials/HollowCircularProgressBar
     tiling=NoTile
     imageSize=<36 36>
     margin=<ltrb 0 0 0 0>
     outlineSettings {
      bUseBrushTransparency=false
      roundingType=HalfHeightRadius
      width=0
      color {colorUseRule=UseColor_Specified, specifiedColor=<rgba 0 0 0 0>}
      cornerRadii=<0 0 0 0>
     }
     tintColor {colorUseRule=UseColor_Specified, specifiedColor=<rgba 1 1 0 1>}
     uVRegion {bIsValid=false, max=<0 0>, min=<0 0>}
    }
   }
   slot {class=OverlaySlot, properties {verticalAlignment=VAlign_Center}}
  }
 ]
}
