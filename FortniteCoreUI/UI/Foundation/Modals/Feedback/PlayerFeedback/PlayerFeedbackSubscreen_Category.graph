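# Category subscreen widget for the player feedback modal, laying out a vertical list of dropdown category buttons.
# A transparent, padded CommonBorder wraps the inherited DynamicEntryBox Buttons_DropdownCategories, which creates evenly spaced PlayerFeedbackSubscreen_DropdownButton entries.
# The event graph holds only unconnected stub events, so any behavior comes from the native parent FortPlayerFeedbackModalSubscreen_CategoryDisplay.
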WidgetBlueprint /FortniteCoreUI/UI/Foundation/Modals/Feedback/PlayerFeedback/PlayerFeedbackSubscreen_Category
graphs [EventGraph {nodes [
   Event PreConstruct()
   Event Construct()
   Event Tick()
  ]}]
variables []
widgets {
 info {
  inheritedWidgetCount=1
  parentClass=/Script/FortniteUI.FortPlayerFeedbackModalSubscreen_CategoryDisplay
  rootWidgetClass=&CommonBorder
  widgetCount=2
 }
 widgets [
  CommonBorder_2 {
   widgetClassPath=&CommonBorder
   widget=CommonBorder_2
   properties {
    background {
     bIsDynamicallyLoaded=false
     drawAs=Box
     imageType=NoImage
     mirroring=NoMirror
     resourceName=None
     resourceObject=null
     tiling=NoTile
     imageSize=<64 64>
     margin=<ltrb 0.25 0.25 0.25 0.25>
     outlineSettings {
      bUseBrushTransparency=false
      roundingType=HalfHeightRadius
      width=0
      color {colorUseRule=UseColor_Specified, specifiedColor=<rgba 0 0 0 0>}
      cornerRadii=<0 0 0 0>
     }
     tintColor {colorUseRule=UseColor_Specified, specifiedColor=<rgba 1 1 1 1>}
     uVRegion {bIsValid=false, max=<0 0>, min=<0 0>}
    }
    brushColor=<rgba 1 1 1 0>
    padding=<ltrb 16 16 16 16>
   }
  }
  Buttons_DropdownCategories {
   widgetClassPath=&DynamicEntryBox
   bInherited=true
   bIsVariable=true
   parent=CommonBorder_2
   widget=Buttons_DropdownCategories
   properties {
    entryBoxType=Vertical
    entryWidgetClass=&/PlayerFeedbackSubscreen_DropdownButton.PlayerFeedbackSubscreen_DropdownButton_C
    numDesignerPreviewEntries=10
    entrySpacing=<0 16>
   }
   slot {class=BorderSlot, properties {padding=<ltrb 16 16 16 16>}}
  }
 ]
}
