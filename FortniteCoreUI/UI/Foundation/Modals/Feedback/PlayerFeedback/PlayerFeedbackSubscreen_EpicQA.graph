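# Player feedback modal subscreen for Epic QA, with pages for JIRA sign-in to submit a FORT bug, sign-in progress, and bug component selection.
# OnStateUpdated maps the QA state to a switcher page and shows the "Sign-In Failed" text only in the failed state, which reuses the sign-in page.
# Username edits are trimmed to the runtime-options character limit, playing a shake animation on trim, guarded by a flag reset after a short delay.
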WidgetBlueprint /FortniteCoreUI/UI/Foundation/Modals/Feedback/PlayerFeedback/PlayerFeedbackSubscreen_EpicQA
graphs [EventGraph {nodes [
   "OnTextChanged (Editable Text) (EditableText_Username)"!() -> (OutputDelegate, then, Text:"Text (by ref)") {is=title K2Node_ComponentBoundEvent}
   > n1 = LimitTextLength!(
    execute
    EditableTextWidget:Widget Object Reference=Get EditableText_Username() -> (EditableText_Username:Editable Text Object Reference)
    Length:int=Get MaxCharactersInTextEntry!(self:Fort Runtime Options Object Reference=GetRuntimeOptions()) -> (MaxCharactersInTextEntry:int=0)
   ) -> (then, WasTrimmed:bool)
   n2 = Get bShakeAnimationIsPlaying() -> (bShakeAnimationIsPlaying:bool)
   > Branch(Condition=AND Boolean(A=@n1, B=NOT Boolean(A=@n2)))
   > Set bShakeAnimationIsPlaying(bShakeAnimationIsPlaying:bool=true) -> (Output_Get:bool)
   > PlayAnimationByName!(
    execute
    self:Fort Simple Widget Animations Panel Object Reference=Get SimpleWidgetAnimationsPanel() -> (SimpleWidgetAnimationsPanel:Fort Simple Widget Animations Panel Object Reference)
    Name:"Name (by ref)"=Get ShakeAnimation() -> (ShakeAnimation:Name=None)
    Speed:float=1
    StartAtTime:float=0
   ) -> (then, ReturnValue:bool)
   > Delay(Duration=0.5)
   > Set bShakeAnimationIsPlaying(bShakeAnimationIsPlaying=false)
   n3 = Event OnStateUpdated!() -> (OutputDelegate, then, NewState:EPlayerFeedback_EpicQAState Enum=SignInPage) {is=title K2Node_Event}
   n4 = Get BugCreationSignInVBox() -> (BugCreationSignInVBox:Vertical Box Object Reference)
   > SetActiveWidget(
    self:Widget Switcher Object Reference=Get Switcher_Main() -> (Switcher_Main:Common Widget Switcher Legacy Object Reference)
    Widget=Select!(
     SignInPage:Widget Object Reference=@n4
     SigningInFailed:Widget Object Reference=@n4
     SigningIn:Widget Object Reference=Get ProgressVBox() -> (ProgressVBox:Vertical Box Object Reference)
     SelectBugComponent:Widget Object Reference=Get BugCreationChooseComponentVBox() -> (BugCreationChooseComponentVBox:Vertical Box Object Reference)
     Index:EPlayerFeedback_EpicQAState Enum=@n3.NewState
    ) -> (ReturnValue:Widget Object Reference)
   )
   > SetVisibility(
    self=Get CommonTextBlock_72() -> (CommonTextBlock_72:Common Text Object Reference)
    InVisibility=Select(
     SignInPage:ESlateVisibility Enum=Collapsed
     SigningInFailed:ESlateVisibility Enum=SelfHitTestInvisible
     SigningIn:ESlateVisibility Enum=Collapsed
     SelectBugComponent:ESlateVisibility Enum=Collapsed
     Index=@n3.NewState
    ) -> (ReturnValue:ESlateVisibility Enum)
   )
  ]}]
variables [
 bShakeAnimationIsPlaying {type=boolean}
 ShakeAnimation {type=string, default=Shake}
]
widgets {
 info {
  inheritedWidgetCount=5
  parentClass=/Script/FortniteUI.FortPlayerFeedbackModalSubscreen_EpicQA
  rootWidgetClass=&Overlay
  widgetCount=30
 }
 widgets [
  Overlay_17 {widgetClassPath=&Overlay, widget=Overlay_17}
  Switcher_Main {
   widgetClassPath=&CommonWidgetSwitcherLegacy
   bIsVariable=true
   parent=Overlay_17
   widget=Switcher_Main
   slot {class=OverlaySlot, properties {horizontalAlignment=HAlign_Fill, verticalAlignment=VAlign_Fill}}
  }
  BugCreationSignInVBox {
   widgetClassPath=&VerticalBox
   bIsVariable=true
   parent=Switcher_Main
   widget=BugCreationSignInVBox
   properties {navigation=&self:WidgetTree.BugCreationSignInVBox.WidgetNavigation_1}
   slot {class=WidgetSwitcherSlot}
  }
  VerticalBox_6 {
   widgetClassPath=&VerticalBox
   parent=BugCreationSignInVBox
   widget=VerticalBox_6
   slot {class=VerticalBoxSlot, properties {padding=<ltrb 32 0 32 16>, size {sizeRule=Fill, value=1}}}
  }
  CommonTextBlock_4 {
   widgetClassPath=&CommonTextBlock
   parent=VerticalBox_6
   widget=CommonTextBlock_4
   properties {
    text=Sign in to JIRA to submit FORT bug
    colorAndOpacity {colorUseRule=UseColor_Specified, specifiedColor=<rgba 0.5149 0.7835 0.9647 1>}
    font {
     bForceMonospaced=false
     bMaterialIsStencil=false
     fontMaterial=null
     fontObject=&BurbankSmall
     letterSpacing=0
     monospacedWidth=1
     size=18
     skewAmount=0
     typefaceFontName=Bold
     outlineSettings {
      bApplyOutlineToDropShadows=false
      bMiteredCorners=false
      bSeparateFillAlpha=false
      outlineMaterial=null
      outlineSize=0
      outlineColor=<rgba 0 0 0 1>
     }
    }
    shadowOffset=<0 0>
   }
   slot {class=VerticalBoxSlot, properties {horizontalAlignment=HAlign_Center, padding=<ltrb 0 20 0 20>}}
  }
  HorizontalBox_487 {
   widgetClassPath=&HorizontalBox
   parent=VerticalBox_6
   widget=HorizontalBox_487
   slot {class=VerticalBoxSlot, properties {padding=<ltrb 10 10 10 10>}}
  }
  TextBlock_110 {
   widgetClassPath=&TextBlock
   parent=HorizontalBox_487
   widget=TextBlock_110
   properties {
    text="User Name: "
    font {
     bForceMonospaced=false
     bMaterialIsStencil=false
     fontMaterial=null
     fontObject=&BurbankSmall
     letterSpacing=0
     monospacedWidth=1
     size=22
     skewAmount=0
     typefaceFontName=Bold
     outlineSettings {
      bApplyOutlineToDropShadows=false
      bMiteredCorners=false
      bSeparateFillAlpha=false
      outlineMaterial=null
      outlineSize=0
      outlineColor=<rgba 0 0 0 1>
     }
    }
   }
   slot {class=HorizontalBoxSlot, properties {verticalAlignment=VAlign_Center}}
  }
  SimpleWidgetAnimationsPanel {
   widgetClassPath=/Script/FortniteUI.FortSimpleWidgetAnimationsPanel
   bIsVariable=true
   parent=HorizontalBox_487
   widget=SimpleWidgetAnimationsPanel
   properties {animations [Shake {
      bIsEnabled=true
      onAnimationsFinished=()
      targets [{
        alpha=null
        bResetOnFinish=true
        bShouldAnimateAlpha=false
        bShouldAnimateScale=false
        bShouldAnimateTranslation=true
        scale=null
        targetName=CommonBorder_8
        translation=/Game/UI/Utilities/Curve_ShakeAnim
        materialScalarAnimations []
       }]
     }]}
   slot {class=HorizontalBoxSlot, properties {size {sizeRule=Fill, value=1}}}
  }
  CommonBorder_8 {
   widgetClassPath=&CommonBorder
   parent=SimpleWidgetAnimationsPanel
   widget=CommonBorder_8
   properties {
    background {
     bIsDynamicallyLoaded=false
     drawAs=Image
     imageType=NoImage
     mirroring=NoMirror
     resourceName=None
     resourceObject=null
     tiling=NoTile
     imageSize=<32 32>
     margin=<ltrb 0 0 0 0>
     outlineSettings {
      bUseBrushTransparency=false
      roundingType=HalfHeightRadius
      width=0
      color {colorUseRule=UseColor_Specified, specifiedColor=<rgba 0 0 0 0>}
      cornerRadii=<0 0 0 0>
     }
     tintColor {colorUseRule=UseColor_Specified, specifiedColor=<rgba 0.0015 0.0331 0.1946 1>}
     uVRegion {bIsValid=false, max=<0 0>, min=<0 0>}
    }
    padding=<ltrb 8 8 8 8>
   }
   slot {class=FortSimpleWidgetAnimationsSlot}
  }
  EditableText_Username {
   widgetClassPath=/Script/UMG.EditableText
   bInherited=true
   bIsVariable=true
   parent=CommonBorder_8
   widget=EditableText_Username
   properties {
    hintText=Username
    widgetStyle {
     backgroundImageComposing {
      bIsDynamicallyLoaded=false
      drawAs=Border
      imageType=FullColor
      mirroring=NoMirror
      resourceName=../../../Engine/Content/Slate/Old/HyperlinkDotted.png
      resourceObject=null
      tiling=Both
      imageSize=<0 0>
      margin=<ltrb 0 0 0 0.1875>
      outlineSettings {
       bUseBrushTransparency=false
       roundingType=HalfHeightRadius
       width=0
       color {colorUseRule=UseColor_Specified, specifiedColor=<rgba 0 0 0 0>}
       cornerRadii=<0 0 0 0>
      }
      tintColor {colorUseRule=UseColor_Specified, specifiedColor=<rgba 1 1 1 1>}
      uVRegion {bIsValid=false, max=<0 0>, min=<0 0>}
     }
     backgroundImageSelected {
      bIsDynamicallyLoaded=false
      drawAs=Box
      imageType=FullColor
      mirroring=NoMirror
      resourceName=../../../Engine/Content/Slate/Common/EditableTextSelectionBackground.png
      resourceObject=null
      tiling=NoTile
      imageSize=<0 0>
      margin=<ltrb 0.25 0.25 0.25 0.25>
      outlineSettings {
       bUseBrushTransparency=false
       roundingType=HalfHeightRadius
       width=0
       color {colorUseRule=UseColor_Specified, specifiedColor=<rgba 0 0 0 0>}
       cornerRadii=<0 0 0 0>
      }
      tintColor {colorUseRule=UseColor_Specified, specifiedColor=<rgba 1 1 1 1>}
      uVRegion {bIsValid=false, max=<0 0>, min=<0 0>}
     }
     caretImage {
      bIsDynamicallyLoaded=false
      drawAs=Image
      imageType=FullColor
      mirroring=NoMirror
      resourceName=../../../Engine/Content/Slate/Old/White.png
      resourceObject=null
      tiling=NoTile
      imageSize=<16 16>
      margin=<ltrb 0 0 0 0>
      outlineSettings {
       bUseBrushTransparency=false
       roundingType=HalfHeightRadius
       width=0
       color {colorUseRule=UseColor_Specified, specifiedColor=<rgba 0 0 0 0>}
       cornerRadii=<0 0 0 0>
      }
      tintColor {colorUseRule=UseColor_Specified, specifiedColor=<rgba 1 1 1 1>}
      uVRegion {bIsValid=false, max=<0 0>, min=<0 0>}
     }
     colorAndOpacity {colorUseRule=UseColor_Specified, specifiedColor=<rgba 1 1 1 1>}
     font {
      bForceMonospaced=false
      bMaterialIsStencil=false
      fontMaterial=null
      fontObject=&BurbankSmall
      letterSpacing=0
      monospacedWidth=1
      size=16
      skewAmount=0
      typefaceFontName=Bold
      outlineSettings {
       bApplyOutlineToDropShadows=false
       bMiteredCorners=false
       bSeparateFillAlpha=false
       outlineMaterial=null
       outlineSize=0
       outlineColor=<rgba 0 0 0 1>
      }
     }
     hintColorAndOpacity {colorUseRule=UseColor_Foreground, specifiedColor=<rgba 1 0 1 1>}
     selectedBackgroundColor {colorUseRule=UseColor_Foreground, specifiedColor=<rgba 1 0 1 1>}
    }
   }
   slot {class=BorderSlot, properties {padding=<ltrb 8 8 8 8>}}
  }
  HorizontalBox_635 {
   widgetClassPath=&HorizontalBox
   parent=VerticalBox_6
   widget=HorizontalBox_635
   slot {class=VerticalBoxSlot, properties {padding=<ltrb 10 10 10 10>}}
  }
  TextBlock_213 {
   widgetClassPath=&TextBlock
   parent=HorizontalBox_635
   widget=TextBlock_213
   properties {
    text="Password: "
    font {
     bForceMonospaced=false
     bMaterialIsStencil=false
     fontMaterial=null
     fontObject=&BurbankSmall
     letterSpacing=0
     monospacedWidth=1
     size=22
     skewAmount=0
     typefaceFontName=Bold
     outlineSettings {
      bApplyOutlineToDropShadows=false
      bMiteredCorners=false
      bSeparateFillAlpha=false
      outlineMaterial=null
      outlineSize=0
      outlineColor=<rgba 0 0 0 1>
     }
    }
   }
   slot {class=HorizontalBoxSlot, properties {verticalAlignment=VAlign_Center, padding=<ltrb 0 0 16 0>}}
  }
  CommonBorder_10 {
   widgetClassPath=&CommonBorder
   parent=HorizontalBox_635
   widget=CommonBorder_10
   properties {
    background {
     bIsDynamicallyLoaded=false
     drawAs=Image
     imageType=NoImage
     mirroring=NoMirror
     resourceName=None
     resourceObject=null
     tiling=NoTile
     imageSize=<32 32>
     margin=<ltrb 0 0 0 0>
     outlineSettings {
      bUseBrushTransparency=false
      roundingType=HalfHeightRadius
      width=0
      color {colorUseRule=UseColor_Specified, specifiedColor=<rgba 0 0 0 0>}
      cornerRadii=<0 0 0 0>
     }
     tintColor {colorUseRule=UseColor_Specified, specifiedColor=<rgba 0.0015 0.0331 0.1946 1>}
     uVRegion {bIsValid=false, max=<0 0>, min=<0 0>}
    }
    padding=<ltrb 8 8 8 8>
   }
   slot {class=HorizontalBoxSlot, properties {size {sizeRule=Fill, value=1}}}
  }
  EditableText_Password {
   widgetClassPath=&EditableText
   bInherited=true
   bIsVariable=true
   parent=CommonBorder_10
   widget=EditableText_Password
   properties {
    hintText=Password
    isPassword=true
    widgetStyle {
     backgroundImageComposing {
      bIsDynamicallyLoaded=false
      drawAs=Border
      imageType=FullColor
      mirroring=NoMirror
      resourceName=../../../Engine/Content/Slate/Old/HyperlinkDotted.png
      resourceObject=null
      tiling=Both
      imageSize=<0 0>
      margin=<ltrb 0 0 0 0.1875>
      outlineSettings {
       bUseBrushTransparency=false
       roundingType=HalfHeightRadius
       width=0
       color {colorUseRule=UseColor_Specified, specifiedColor=<rgba 0 0 0 0>}
       cornerRadii=<0 0 0 0>
      }
      tintColor {colorUseRule=UseColor_Specified, specifiedColor=<rgba 1 1 1 1>}
      uVRegion {bIsValid=false, max=<0 0>, min=<0 0>}
     }
     backgroundImageSelected {
      bIsDynamicallyLoaded=false
      drawAs=Box
      imageType=FullColor
      mirroring=NoMirror
      resourceName=../../../Engine/Content/Slate/Common/EditableTextSelectionBackground.png
      resourceObject=null
      tiling=NoTile
      imageSize=<0 0>
      margin=<ltrb 0.25 0.25 0.25 0.25>
      outlineSettings {
       bUseBrushTransparency=false
       roundingType=HalfHeightRadius
       width=0
       color {colorUseRule=UseColor_Specified, specifiedColor=<rgba 0 0 0 0>}
       cornerRadii=<0 0 0 0>
      }
      tintColor {colorUseRule=UseColor_Specified, specifiedColor=<rgba 1 1 1 1>}
      uVRegion {bIsValid=false, max=<0 0>, min=<0 0>}
     }
     caretImage {
      bIsDynamicallyLoaded=false
      drawAs=Image
      imageType=FullColor
      mirroring=NoMirror
      resourceName=../../../Engine/Content/Slate/Old/White.png
      resourceObject=null
      tiling=NoTile
      imageSize=<16 16>
      margin=<ltrb 0 0 0 0>
      outlineSettings {
       bUseBrushTransparency=false
       roundingType=HalfHeightRadius
       width=0
       color {colorUseRule=UseColor_Specified, specifiedColor=<rgba 0 0 0 0>}
       cornerRadii=<0 0 0 0>
      }
      tintColor {colorUseRule=UseColor_Specified, specifiedColor=<rgba 1 1 1 1>}
      uVRegion {bIsValid=false, max=<0 0>, min=<0 0>}
     }
     colorAndOpacity {colorUseRule=UseColor_Specified, specifiedColor=<rgba 1 1 1 1>}
     font {
      bForceMonospaced=false
      bMaterialIsStencil=false
      fontMaterial=null
      fontObject=&BurbankSmall
      letterSpacing=0
      monospacedWidth=1
      size=16
      skewAmount=0
      typefaceFontName=Bold
      outlineSettings {
       bApplyOutlineToDropShadows=false
       bMiteredCorners=false
       bSeparateFillAlpha=false
       outlineMaterial=null
       outlineSize=0
       outlineColor=<rgba 0 0 0 1>
      }
     }
     hintColorAndOpacity {colorUseRule=UseColor_Foreground, specifiedColor=<rgba 1 0 1 1>}
     selectedBackgroundColor {colorUseRule=UseColor_Foreground, specifiedColor=<rgba 1 0 1 1>}
    }
   }
   slot {class=BorderSlot, properties {padding=<ltrb 8 8 8 8>}}
  }
  CommonTextBlock_72 {
   widgetClassPath=&CommonTextBlock
   bIsVariable=true
   parent=VerticalBox_6
   widget=CommonTextBlock_72
   properties {
    text=Sign-In Failed
    colorAndOpacity {colorUseRule=UseColor_Specified, specifiedColor=<rgba 0.8542 0.1491 0.2686 1>}
    font {
     bForceMonospaced=false
     bMaterialIsStencil=false
     fontMaterial=null
     fontObject=&BurbankSmall
     letterSpacing=0
     monospacedWidth=1
     size=24
     skewAmount=0
     typefaceFontName=Bold
     outlineSettings {
      bApplyOutlineToDropShadows=false
      bMiteredCorners=false
      bSeparateFillAlpha=false
      outlineMaterial=null
      outlineSize=0
      outlineColor=<rgba 0 0 0 1>
     }
    }
    shadowOffset=<0 0>
   }
   slot {class=VerticalBoxSlot, properties {horizontalAlignment=HAlign_Center, padding=<ltrb 0 20 0 0>}}
  }
  HorizontalBox_7 {
   widgetClassPath=&HorizontalBox
   parent=BugCreationSignInVBox
   widget=HorizontalBox_7
   slot {class=VerticalBoxSlot, properties {horizontalAlignment=HAlign_Center, padding=<ltrb 0 0 0 20>}}
  }
  Button_SignIn {
   widgetClassPath=/FortUILibrary/Foundation/Buttons/SoloButton
   bInherited=true
   bIsVariable=true
   parent=HorizontalBox_7
   widget=Button_SignIn
   properties {
    button Description=Send
    clickMethod=DownAndUp
    minWidth=192
    onButtonBaseClicked=(&self:WidgetTree.Button_SignIn.HandleOnButtonClicked)
    onButtonBaseDoubleClicked=(&self:WidgetTree.Button_SignIn.HandleOnButtonDoubleClicked)
    onButtonBaseHovered=(&self:WidgetTree.Button_SignIn.HandleOnButtonHovered)
    onButtonBaseUnhovered=(&self:WidgetTree.Button_SignIn.HandleOnButtonUnhovered)
    onSelectedChangedBase=(&self:WidgetTree.Button_SignIn.HandleOnSelectedChanged)
    pressMethod=DownAndUp
    hoveredSlateSoundOverride {resourceObject=/Game/Sounds/Fort_UI_Sounds/FrontEnd/New_FrontEnd/Fortnite_UI_MainMenu_Tab_Hover_01_Cue}
    pressedSlateSoundOverride {resourceObject=/Game/Sounds/Fort_UI_Sounds/FrontEnd/New_FrontEnd/Fortnite_UI_MainMenu_Generic_01_Cue}
    triggeringInputAction {dataTable=&InputActionDataTable, rowName=GenericFaceButtonLeft}
   }
   slot {class=HorizontalBoxSlot, properties {padding=<ltrb 0 0 16 0>}}
  }
  BugCreationChooseComponentVBox {
   widgetClassPath=&VerticalBox
   bIsVariable=true
   parent=Switcher_Main
   widget=BugCreationChooseComponentVBox
   slot {class=WidgetSwitcherSlot}
  }
  VerticalBox_294 {
   widgetClassPath=&VerticalBox
   parent=BugCreationChooseComponentVBox
   widget=VerticalBox_294
   slot {class=VerticalBoxSlot, properties {padding=<ltrb 32 0 32 16>, size {sizeRule=Fill, value=1}}}
  }
  CommonTextBlock_518 {
   widgetClassPath=&CommonTextBlock
   parent=VerticalBox_294
   widget=CommonTextBlock_518
   properties {
    text=Select Bug Component
    colorAndOpacity {colorUseRule=UseColor_Specified, specifiedColor=<rgba 0.5149 0.7835 0.9647 1>}
    font {
     bForceMonospaced=false
     bMaterialIsStencil=false
     fontMaterial=null
     fontObject=&BurbankSmall
     letterSpacing=0
     monospacedWidth=1
     size=18
     skewAmount=0
     typefaceFontName=Bold
     outlineSettings {
      bApplyOutlineToDropShadows=false
      bMiteredCorners=false
      bSeparateFillAlpha=false
      outlineMaterial=null
      outlineSize=0
      outlineColor=<rgba 0 0 0 1>
     }
    }
    shadowOffset=<0 0>
   }
   slot {class=VerticalBoxSlot, properties {horizontalAlignment=HAlign_Center, padding=<ltrb 0 20 0 20>}}
  }
  HorizontalBox_261 {widgetClassPath=&HorizontalBox, parent=VerticalBox_294, widget=HorizontalBox_261, slot {class=VerticalBoxSlot}}
  TextBlock_185 {
   widgetClassPath=&TextBlock
   parent=HorizontalBox_261
   widget=TextBlock_185
   properties {
    text="Bug Component: "
    font {
     bForceMonospaced=false
     bMaterialIsStencil=false
     fontMaterial=null
     fontObject=&BurbankSmall
     letterSpacing=0
     monospacedWidth=1
     size=20
     skewAmount=0
     typefaceFontName=Bold
     outlineSettings {
      bApplyOutlineToDropShadows=false
      bMiteredCorners=false
      bSeparateFillAlpha=false
      outlineMaterial=null
      outlineSize=0
      outlineColor=<rgba 0 0 0 1>
     }
    }
   }
   slot {class=HorizontalBoxSlot, properties {padding=<ltrb 0 0 20 0>}}
  }
  ComboBoxString_BugComponents {
   widgetClassPath=/Script/UMG.ComboBoxString
   bInherited=true
   bIsVariable=true
   parent=HorizontalBox_261
   widget=ComboBoxString_BugComponents
   slot {class=HorizontalBoxSlot, properties {size {sizeRule=Fill, value=1}}}
  }
  CommonBorder_11 {
   widgetClassPath=&CommonBorder
   parent=BugCreationChooseComponentVBox
   widget=CommonBorder_11
   properties {
    horizontalAlignment=HAlign_Right
    style=/FortUILibrary/Styles/Border/Border_Solid_DkBlue.Border_Solid_DkBlue_C
    brushColor=<rgba 1 1 1 0.3>
    padding=<ltrb 32 16 32 16>
   }
   slot {class=VerticalBoxSlot}
  }
  HorizontalBox_3 {
   widgetClassPath=&HorizontalBox
   parent=CommonBorder_11
   widget=HorizontalBox_3
   slot {class=BorderSlot, properties {horizontalAlignment=HAlign_Right, padding=<ltrb 32 16 32 16>}}
  }
  Button_ConfirmComponent {
   widgetClassPath=&IconTextButton
   bInherited=true
   bIsVariable=true
   parent=HorizontalBox_3
   widget=Button_ConfirmComponent
   properties {
    buttonText=""
    contentAlignment=HAlign_Center
    minWidth=192
    onButtonBaseClicked=(&self:WidgetTree.Button_ConfirmComponent.HandleOnButtonClicked)
    onButtonBaseDoubleClicked=(&self:WidgetTree.Button_ConfirmComponent.HandleOnButtonDoubleClicked)
    onButtonBaseHovered=(&self:WidgetTree.Button_ConfirmComponent.HandleOnButtonHovered)
    onButtonBaseUnhovered=(&self:WidgetTree.Button_ConfirmComponent.HandleOnButtonUnhovered)
    onSelectedChangedBase=(&self:WidgetTree.Button_ConfirmComponent.HandleOnSelectedChanged)
    style=/FortUILibrary/Foundation/Buttons/ButtonStyles/TopBar/ButtonStyle-Outline.ButtonStyle-Outline_C
    triggeringInputAction {dataTable=&InputActionDataTable, rowName=Send}
   }
   slot {class=HorizontalBoxSlot, properties {padding=<ltrb 0 0 16 0>}}
  }
  CancelBug {
   widgetClassPath=&IconTextButton
   bIsVariable=true
   parent=HorizontalBox_3
   widget=CancelBug
   properties {
    buttonText=""
    contentAlignment=HAlign_Center
    minWidth=192
    onButtonBaseClicked=(&self:WidgetTree.CancelBug.HandleOnButtonClicked)
    onButtonBaseDoubleClicked=(&self:WidgetTree.CancelBug.HandleOnButtonDoubleClicked)
    onButtonBaseHovered=(&self:WidgetTree.CancelBug.HandleOnButtonHovered)
    onButtonBaseUnhovered=(&self:WidgetTree.CancelBug.HandleOnButtonUnhovered)
    onSelectedChangedBase=(&self:WidgetTree.CancelBug.HandleOnSelectedChanged)
    style=&ButtonStyle-Outline_C
    triggeringInputAction {dataTable=&InputActionDataTable, rowName=Cancel}
   }
   slot {class=HorizontalBoxSlot}
  }
  ProgressVBox {
   widgetClassPath=&VerticalBox
   bIsVariable=true
   parent=Switcher_Main
   widget=ProgressVBox
   slot {class=WidgetSwitcherSlot, properties {horizontalAlignment=HAlign_Center, verticalAlignment=VAlign_Center}}
  }
  CommonTextBlock_10 {
   widgetClassPath=&CommonTextBlock
   parent=ProgressVBox
   widget=CommonTextBlock_10
   properties {
    text=Signing In...
    textTransformPolicy=ToUpper
    colorAndOpacity {colorUseRule=UseColor_Specified, specifiedColor=<rgba 0.5149 0.7835 0.9647 1>}
    font {
     bForceMonospaced=false
     bMaterialIsStencil=false
     fontMaterial=null
     fontObject=&BurbankSmall
     letterSpacing=0
     monospacedWidth=1
     size=26
     skewAmount=0
     typefaceFontName=Bold
     outlineSettings {
      bApplyOutlineToDropShadows=false
      bMiteredCorners=false
      bSeparateFillAlpha=false
      outlineMaterial=null
      outlineSize=0
      outlineColor=<rgba 0 0 0 1>
     }
    }
    shadowOffset=<0 0>
   }
   slot {class=VerticalBoxSlot, properties {padding=<ltrb 0 0 0 20>}}
  }
  Image {
   widgetClassPath=&Image
   bIsVariable=true
   parent=ProgressVBox
   widget=Image
   properties {brush {
     bIsDynamicallyLoaded=false
     drawAs=Image
     imageType=NoImage
     mirroring=NoMirror
     resourceName=None
     resourceObject=/FortUILibrary/UIKit/Throbber/MI_UIKit_Color_Throbber
     tiling=NoTile
     imageSize=<64 64>
     margin=<ltrb 0 0 0 0>
     outlineSettings {
      bUseBrushTransparency=false
      roundingType=HalfHeightRadius
      width=0
      color {colorUseRule=UseColor_Specified, specifiedColor=<rgba 0 0 0 0>}
      cornerRadii=<0 0 0 0>
     }
     tintColor {colorUseRule=UseColor_Specified, specifiedColor=<rgba 1 1 1 1>}
     uVRegion {bIsValid=false, max=<0 0>, min=<0 0>}
    }}
   slot {class=VerticalBoxSlot, properties {horizontalAlignment=HAlign_Center, verticalAlignment=VAlign_Center}}
  }
 ]
}
