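# Player-select subscreen for the player feedback modal, where players available to report are listed.
# The native FortPlayerFeedbackModalSubscreen_PlayerSelect parent fires OnCreatedPlayerList with a bHasPlayers flag.
# The event toggles visibility between a centered "No Players to Report" message and a vertical DynamicEntryBox of PlayerButton entries.
# The Blueprint handles only layout and this switch; no list-filling logic appears in its graph.
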WidgetBlueprint /FortniteCoreUI/UI/Foundation/Modals/Feedback/PlayerFeedback/PlayerFeedbackSubscreen_PlayerSelect
graphs [EventGraph {nodes [
   n1 = Event OnCreatedPlayerList!() -> (OutputDelegate, then, bHasPlayers:bool) {is=title K2Node_Event}
   > SetVisibility(
    self=Get NoPlayers() -> (NoPlayers:Vertical Box Object Reference)
    InVisibility=Select(Option 0=Visible, Option 1=Collapsed, Index=@n1.bHasPlayers)
   )
   > SetVisibility(
    self=Get HasPlayers() -> (HasPlayers:Overlay Object Reference)
    InVisibility=Select(Option 0=Collapsed, Option 1=Visible, Index=@n1.bHasPlayers)
   )
  ]}]
variables []
widgets {
 info {
  inheritedWidgetCount=2
  parentClass=/Script/FortniteUI.FortPlayerFeedbackModalSubscreen_PlayerSelect
  rootWidgetClass=&CommonBorder
  widgetCount=6
 }
 widgets [
  CommonBorder_2 {
   widgetClassPath=&CommonBorder
   widget=CommonBorder_2
   properties {
    background {
     bIsDynamicallyLoaded=false
     drawAs=Box
     imageType=NoImage
     mirroring=NoMirror
     resourceName=None
     resourceObject=null
     tiling=NoTile
     imageSize=<64 64>
     margin=<ltrb 0.25 0.25 0.25 0.25>
     outlineSettings {
      bUseBrushTransparency=false
      roundingType=HalfHeightRadius
      width=0
      color {colorUseRule=UseColor_Specified, specifiedColor=<rgba 0 0 0 0>}
      cornerRadii=<0 0 0 0>
     }
     tintColor {colorUseRule=UseColor_Specified, specifiedColor=<rgba 1 1 1 1>}
     uVRegion {bIsValid=false, max=<0 0>, min=<0 0>}
    }
    brushColor=<rgba 1 1 1 0>
    padding=<ltrb 16 16 16 16>
   }
  }
  Container {
   widgetClassPath=&Overlay
   parent=CommonBorder_2
   widget=Container
   slot {class=BorderSlot, properties {padding=<ltrb 16 16 16 16>}}
  }
  NoPlayers {
   widgetClassPath=&VerticalBox
   bIsVariable=true
   parent=Container
   widget=NoPlayers
   slot {class=OverlaySlot, properties {horizontalAlignment=HAlign_Center, verticalAlignment=VAlign_Center}}
  }
  Text_NoPlayers {
   widgetClassPath=&CommonTextBlock
   bInherited=true
   parent=NoPlayers
   widget=Text_NoPlayers
   properties {
    autoWrapText=true
    justification=Center
    text=No Players to Report
    wrapTextAt=800
    font {
     bForceMonospaced=false
     bMaterialIsStencil=false
     fontMaterial=null
     fontObject=&HeadingNow
     letterSpacing=0
     monospacedWidth=1
     size=20.25
     skewAmount=0
     typefaceFontName=75 Medium
     outlineSettings {
      bApplyOutlineToDropShadows=false
      bMiteredCorners=false
      bSeparateFillAlpha=false
      outlineMaterial=null
      outlineSize=0
      outlineColor=<rgba 0 0 0 1>
     }
    }
    shadowOffset=<0 0>
   }
   slot {class=VerticalBoxSlot, properties {verticalAlignment=VAlign_Top}}
  }
  HasPlayers {
   widgetClassPath=&Overlay
   bIsVariable=true
   parent=Container
   widget=HasPlayers
   slot {class=OverlaySlot, properties {horizontalAlignment=HAlign_Fill, verticalAlignment=VAlign_Fill}}
  }
  Buttons_Players {
   widgetClassPath=&DynamicEntryBox
   bInherited=true
   bIsVariable=true
   parent=HasPlayers
   widget=Buttons_Players
   properties {
    entryBoxType=Vertical
    entryWidgetClass=&/PlayerFeedbackSubscreen_PlayerButton.PlayerFeedbackSubscreen_PlayerButton_C
    numDesignerPreviewEntries=10
    entrySizeRule {sizeRule=Fill, value=1}
    entrySpacing=<0 16>
   }
   slot {class=OverlaySlot, properties {horizontalAlignment=HAlign_Fill, verticalAlignment=VAlign_Fill}}
  }
 ]
}
