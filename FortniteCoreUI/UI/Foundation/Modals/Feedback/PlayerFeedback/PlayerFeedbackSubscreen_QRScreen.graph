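# Player feedback modal subscreen that shows a QR code and report details for reporting potentially illegal content.
# Setter events for player, island, community post and Aurora item or offer set or hide the title, target name and reporter-anonymity footnote.
# The footnote only appears when the ClamOne runtime option is enabled.
# SetQRCodeImage stores the Aurora-purchased flag and encodes GetURL into the QR generator, whose wrapping button launches that URL on click.
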WidgetBlueprint /FortniteCoreUI/UI/Foundation/Modals/Feedback/PlayerFeedback/PlayerFeedbackSubscreen_QRScreen
graphs [
 EventGraph {nodes [
   "On Clicked (QRImage)"!() -> (OutputDelegate, then, Button:Common Button Base Object Reference) {is=title K2Node_ComponentBoundEvent}
   > LaunchURL!(execute, URL:String=GetURL!(self:Fort Player Feedback Modal Subscreen QRScreen Object Reference) -> (ReturnValue:String)) -> (then)
   n1 = Event SetPlayerInfo!() -> (OutputDelegate, then, PlayerName:"Text (by ref)", PlayerAltName:"Text (by ref)", bIsPlayerCreator:bool) {is=title K2Node_Event}
   > SetText(
    self=Get Text_TargetName() -> (Text_TargetName:Common Rich Text Block Object Reference)
    InText=SelectText(
     A=@n1.PlayerName
     B=Format Text!(Format="LOCTEXT(\"{0}\\r\\n{1}\")", 0:"Text (by ref)"=@n1.PlayerName, 1:"Text (by ref)"=@n1.PlayerAltName) -> (Result)
     bPickA=TextIsEmpty(InText=@n1.PlayerAltName)
    )
   )
   > SetVisibility(self=Get Text_TargetName(), InVisibility=SelfHitTestInvisible)
   > SetClamOneFootnoteVisibility!(execute, self, InFootnoteText:Text=Format Text!(Format="LOCTEXT(\"They will not be notified that you are the reporter.\")") -> (Result)) -> (then)
   > SetText(
    self=Get Text_Title() -> (Text_Title:Common Rich Text Block Object Reference)
    InText="LOCTEXT(\"HOW CAN I REPORT POTENTIALLY ILLEGAL CONTENT?\")"
   )
   > SetQRCodeImage!(execute, self, bIsAuroraPurchasedItem:bool=false) -> (then)
   n2 = Event SetIslandInfo!() -> (OutputDelegate, then, IslandCode:"Text (by ref)") {is=title K2Node_Event}
   > SetText(self=Get Text_TargetName(), InText=Format Text!(Format="LOCTEXT(\"{0}\")", 0:"Text (by ref)"=@n2.IslandCode) -> (Result))
   > SetVisibility(self=Get Text_TargetName(), InVisibility=SelfHitTestInvisible)
   > SetClamOneFootnoteVisibility(InFootnoteText=Format Text!(Format="LOCTEXT(\"The Island Creator will not be notified that you are the reporter.\")") -> (Result))
   > SetText(self=Get Text_Title(), InText="LOCTEXT(\"HOW CAN I REPORT POTENTIALLY ILLEGAL CONTENT?\")")
   > SetQRCodeImage(bIsAuroraPurchasedItem=false)
   n3 = Event SetCommunityPostInfo!() -> (OutputDelegate, then, PostId:String) {is=title K2Node_Event}
   > SetText(self=Get Text_TargetName(), InText="To Text (String)"(InString=@n3.PostId))
   > SetVisibility(self=Get Text_TargetName(), InVisibility=SelfHitTestInvisible)
   > SetClamOneFootnoteVisibility(InFootnoteText=Format Text(Format="LOCTEXT(\"They will not be notified that you are the reporter.\")"))
   > SetText(self=Get Text_Title(), InText="LOCTEXT(\"HOW CAN I REPORT POTENTIALLY ILLEGAL CONTENT?\")")
   > SetQRCodeImage(bIsAuroraPurchasedItem=false)
   Event SetAuroraPurchasedInIslandItemInfo() {is=title K2Node_Event}
   > SetVisibility(self=Get Text_TargetName(), InVisibility=Collapsed)
   > SetClamOneFootnoteVisibility()
   > SetVisibility(self=Get Text_Title(), InVisibility=Collapsed)
   > SetQRCodeImage(bIsAuroraPurchasedItem=true)
   Event SetAuroraOfferInfo!() -> (OutputDelegate, then, OfferId:"Text (by ref)") {is=title K2Node_Event}
   > SetVisibility(self=Get Text_TargetName(), InVisibility=Collapsed)
   > SetClamOneFootnoteVisibility(InFootnoteText=Format Text(Format="LOCTEXT(\"The creator won\\'t be notified that you\\'re the reporter.\")"))
   > SetVisibility(self=Get Text_Title(), InVisibility=Collapsed)
   > SetQRCodeImage(bIsAuroraPurchasedItem=false)
  ]}
 SetClamOneFootnoteVisibility {nodes [
   n1 = SetClamOneFootnoteVisibility!() -> (then, InFootnoteText:Text) {is=title K2Node_FunctionEntry}
   n2 = Get Text_Footnote() -> (Text_Footnote:Common Rich Text Block Object Reference)
   n3 = GetIsClamOneEnabled!(self:Fort Runtime Options Object Reference=GetRuntimeOptions()) -> (ReturnValue:bool)
   > Set Visibility(Visibility=Select(Option 0=Collapsed, Option 1=SelfHitTestInvisible, Index=@n3), self=@n2)
   > Branch(Condition=@n3)
   > SetText(self=@n2, InText=@n1)
  ]}
 SetQRCodeImage {nodes [
   n1 = SetQRCodeImage!() -> (then, bIsAuroraPurchasedItem:bool=~) {is=title K2Node_FunctionEntry}
   > Set bIsAuroraPurchasedItem(bIsAuroraPurchasedItem:bool=@n1) -> (Output_Get:bool)
   > SetEncodedString!(
    execute
    self:WBP QRCode Generator Object Reference=Get QRCodeImage() -> (QRCodeImage:WBP QRCode Generator Object Reference)
    StringToEncode:String=GetURL()
   ) -> (then)
  ]}
]
variables [bIsAuroraPurchasedItem {type=boolean}]
widgets {
 info {
  inheritedWidgetCount=1
  namedSlotCount=1
  parentClass=/Script/FortniteUI.FortPlayerFeedbackModalSubscreen_QRScreen
  rootWidgetClass=&SizeBox
  widgetCount=11
 }
 widgets [
  Confirmation {
   widgetClassPath=&SizeBox
   bIsVariable=true
   widget=Confirmation
   properties {bOverride_MaxDesiredWidth=true, maxDesiredWidth=1000}
  }
  HorizontalBox_0 {widgetClassPath=&HorizontalBox, parent=Confirmation, widget=HorizontalBox_0, slot {class=SizeBoxSlot}}
  VerticalBox_0 {
   widgetClassPath=&VerticalBox
   parent=HorizontalBox_0
   widget=VerticalBox_0
   slot {class=HorizontalBoxSlot, properties {size {sizeRule=Fill, value=1}}}
  }
  Text_Title {
   widgetClassPath=&CommonRichTextBlock
   bIsVariable=true
   parent=VerticalBox_0
   widget=Text_Title
   properties {
    autoWrapText=true
    bOverrideDefaultStyle=true
    text=HOW CAN I REPORT POTENTIALLY ILLEGAL CONTENT?
    defaultTextStyleOverride {
     overflowPolicy=Clip
     transformPolicy=None
     colorAndOpacity {colorUseRule=UseColor_Specified, specifiedColor=<rgba 1 1 1 1>}
     font {
      bForceMonospaced=false
      bMaterialIsStencil=false
      fontMaterial=null
      fontObject=&HeadingNow
      letterSpacing=0
      monospacedWidth=1
      size=20.25
      skewAmount=0
      typefaceFontName=75 Medium
      outlineSettings {
       bApplyOutlineToDropShadows=false
       bMiteredCorners=false
       bSeparateFillAlpha=false
       outlineMaterial=null
       outlineSize=0
       outlineColor=<rgba 0 0 0 1>
      }
     }
     highlightColor {colorUseRule=UseColor_Specified, specifiedColor=<rgba 0 0 0 1>}
     highlightShape {
      bIsDynamicallyLoaded=false
      drawAs=Image
      imageType=NoImage
      mirroring=NoMirror
      resourceName=None
      resourceObject=null
      tiling=NoTile
      imageSize=<32 32>
      margin=<ltrb 0 0 0 0>
      outlineSettings {
       bUseBrushTransparency=false
       roundingType=HalfHeightRadius
       width=0
       color {colorUseRule=UseColor_Specified, specifiedColor=<rgba 0 0 0 0>}
       cornerRadii=<0 0 0 0>
      }
      tintColor {colorUseRule=UseColor_Specified, specifiedColor=<rgba 1 1 1 1>}
      uVRegion {bIsValid=false, max=<0 0>, min=<0 0>}
     }
     selectedBackgroundColor {colorUseRule=UseColor_Foreground, specifiedColor=<rgba 1 0 1 1>}
     shadowColorAndOpacity=<rgba 0 0 0 1>
     shadowOffset=<0 0>
     strikeBrush {
      bIsDynamicallyLoaded=false
      drawAs=NoDrawType
      imageType=NoImage
      mirroring=NoMirror
      resourceName=None
      resourceObject=null
      tiling=NoTile
      imageSize=<32 32>
      margin=<ltrb 0 0 0 0>
      outlineSettings {
       bUseBrushTransparency=false
       roundingType=HalfHeightRadius
       width=0
       color {colorUseRule=UseColor_Specified, specifiedColor=<rgba 0 0 0 0>}
       cornerRadii=<0 0 0 0>
      }
      tintColor {colorUseRule=UseColor_Specified, specifiedColor=<rgba 1 1 1 1>}
      uVRegion {bIsValid=false, max=<0 0>, min=<0 0>}
     }
     underlineBrush {
      bIsDynamicallyLoaded=false
      drawAs=Image
      imageType=NoImage
      mirroring=NoMirror
      resourceName=None
      resourceObject=null
      tiling=NoTile
      imageSize=<32 32>
      margin=<ltrb 0 0 0 0>
      outlineSettings {
       bUseBrushTransparency=false
       roundingType=HalfHeightRadius
       width=0
       color {colorUseRule=UseColor_Specified, specifiedColor=<rgba 0 0 0 0>}
       cornerRadii=<0 0 0 0>
      }
      tintColor {colorUseRule=UseColor_Specified, specifiedColor=<rgba 1 1 1 1>}
      uVRegion {bIsValid=false, max=<0 0>, min=<0 0>}
     }
    }
   }
   slot {class=VerticalBoxSlot, properties {padding=<ltrb 0 0 0 12>}}
  }
  Text_Description {
   widgetClassPath=/Script/FortniteUI.FortRichTextBlockLegacy
   bInherited=true
   bIsVariable=true
   parent=VerticalBox_0
   widget=Text_Description
   properties {
    autoWrapText=true
    styleSet=/Game/UI/Foundation/Modals/Feedback/PlayerFeedback/FeedbackScreenRichTextStyles
    text="[PH]\r\n\r\n1. Step 1\r\n2. Step 2\r\n3. Step 3"
    hyperlinkButtonStyle {
     clickedSlateSound {resourceObject=null}
     disabled {
      bIsDynamicallyLoaded=false
      drawAs=NoDrawType
      imageType=NoImage
      mirroring=NoMirror
      resourceName=None
      resourceObject=null
      tiling=NoTile
      imageSize=<0 0>
      margin=<ltrb 0 0 0 0>
      outlineSettings {
       bUseBrushTransparency=false
       roundingType=HalfHeightRadius
       width=0
       color {colorUseRule=UseColor_Specified, specifiedColor=<rgba 0 0 0 0>}
       cornerRadii=<0 0 0 0>
      }
      tintColor {colorUseRule=UseColor_Specified, specifiedColor=<rgba 1 1 1 1>}
      uVRegion {bIsValid=false, max=<0 0>, min=<0 0>}
     }
     disabledForeground {colorUseRule=UseColor_Foreground, specifiedColor=<rgba 1 0 1 1>}
     hovered {
      bIsDynamicallyLoaded=false
      drawAs=Image
      imageType=NoImage
      mirroring=NoMirror
      resourceName=None
      resourceObject=null
      tiling=NoTile
      imageSize=<32 32>
      margin=<ltrb 0 0 0 0>
      outlineSettings {
       bUseBrushTransparency=false
       roundingType=HalfHeightRadius
       width=0
       color {colorUseRule=UseColor_Specified, specifiedColor=<rgba 0 0 0 0>}
       cornerRadii=<0 0 0 0>
      }
      tintColor {colorUseRule=UseColor_Specified, specifiedColor=<rgba 1 1 1 1>}
      uVRegion {bIsValid=false, max=<0 0>, min=<0 0>}
     }
     hoveredForeground {colorUseRule=UseColor_Foreground, specifiedColor=<rgba 1 0 1 1>}
     hoveredSlateSound {resourceObject=null}
     normal {
      bIsDynamicallyLoaded=false
      drawAs=Image
      imageType=NoImage
      mirroring=NoMirror
      resourceName=None
      resourceObject=null
      tiling=NoTile
      imageSize=<32 32>
      margin=<ltrb 0 0 0 0>
      outlineSettings {
       bUseBrushTransparency=false
       roundingType=HalfHeightRadius
       width=0
       color {colorUseRule=UseColor_Specified, specifiedColor=<rgba 0 0 0 0>}
       cornerRadii=<0 0 0 0>
      }
      tintColor {colorUseRule=UseColor_Specified, specifiedColor=<rgba 1 1 1 0>}
      uVRegion {bIsValid=false, max=<0 0>, min=<0 0>}
     }
     normalForeground {colorUseRule=UseColor_Foreground, specifiedColor=<rgba 1 0 1 1>}
     normalPadding=<ltrb 0 0 0 0>
     pressed {
      bIsDynamicallyLoaded=false
      drawAs=Image
      imageType=NoImage
      mirroring=NoMirror
      resourceName=None
      resourceObject=null
      tiling=NoTile
      imageSize=<32 32>
      margin=<ltrb 0 0 0 0>
      outlineSettings {
       bUseBrushTransparency=false
       roundingType=HalfHeightRadius
       width=0
       color {colorUseRule=UseColor_Specified, specifiedColor=<rgba 0 0 0 0>}
       cornerRadii=<0 0 0 0>
      }
      tintColor {colorUseRule=UseColor_Specified, specifiedColor=<rgba 1 1 1 1>}
      uVRegion {bIsValid=false, max=<0 0>, min=<0 0>}
     }
     pressedForeground {colorUseRule=UseColor_Foreground, specifiedColor=<rgba 1 0 1 1>}
     pressedPadding=<ltrb 0 0 0 0>
     pressedSlateSound {resourceObject=null}
    }
   }
   slot {class=VerticalBoxSlot, properties {padding=<ltrb 0 0 0 24>}}
  }
  Text_TargetName {
   widgetClassPath=&CommonRichTextBlock
   bIsVariable=true
   parent=VerticalBox_0
   widget=Text_TargetName
   properties {
    autoWrapText=true
    bOverrideDefaultStyle=true
    text=TargetNameHere
    wrapTextAt=1000
    defaultTextStyleOverride {
     overflowPolicy=Clip
     transformPolicy=None
     colorAndOpacity {colorUseRule=UseColor_Specified, specifiedColor=<rgba 1 1 1 1>}
     font {
      bForceMonospaced=false
      bMaterialIsStencil=false
      fontMaterial=null
      fontObject=&HeadingNow
      letterSpacing=0
      monospacedWidth=1
      size=20.25
      skewAmount=0
      typefaceFontName=75 Medium
      outlineSettings {
       bApplyOutlineToDropShadows=false
       bMiteredCorners=false
       bSeparateFillAlpha=false
       outlineMaterial=null
       outlineSize=0
       outlineColor=<rgba 0 0 0 1>
      }
     }
     highlightColor {colorUseRule=UseColor_Specified, specifiedColor=<rgba 0 0 0 1>}
     highlightShape {
      bIsDynamicallyLoaded=false
      drawAs=Image
      imageType=NoImage
      mirroring=NoMirror
      resourceName=None
      resourceObject=null
      tiling=NoTile
      imageSize=<32 32>
      margin=<ltrb 0 0 0 0>
      outlineSettings {
       bUseBrushTransparency=false
       roundingType=HalfHeightRadius
       width=0
       color {colorUseRule=UseColor_Specified, specifiedColor=<rgba 0 0 0 0>}
       cornerRadii=<0 0 0 0>
      }
      tintColor {colorUseRule=UseColor_Specified, specifiedColor=<rgba 1 1 1 1>}
      uVRegion {bIsValid=false, max=<0 0>, min=<0 0>}
     }
     selectedBackgroundColor {colorUseRule=UseColor_Foreground, specifiedColor=<rgba 1 0 1 1>}
     shadowColorAndOpacity=<rgba 0 0 0 1>
     shadowOffset=<0 0>
     strikeBrush {
      bIsDynamicallyLoaded=false
      drawAs=NoDrawType
      imageType=NoImage
      mirroring=NoMirror
      resourceName=None
      resourceObject=null
      tiling=NoTile
      imageSize=<32 32>
      margin=<ltrb 0 0 0 0>
      outlineSettings {
       bUseBrushTransparency=false
       roundingType=HalfHeightRadius
       width=0
       color {colorUseRule=UseColor_Specified, specifiedColor=<rgba 0 0 0 0>}
       cornerRadii=<0 0 0 0>
      }
      tintColor {colorUseRule=UseColor_Specified, specifiedColor=<rgba 1 1 1 1>}
      uVRegion {bIsValid=false, max=<0 0>, min=<0 0>}
     }
     underlineBrush {
      bIsDynamicallyLoaded=false
      drawAs=Image
      imageType=NoImage
      mirroring=NoMirror
      resourceName=None
      resourceObject=null
      tiling=NoTile
      imageSize=<32 32>
      margin=<ltrb 0 0 0 0>
      outlineSettings {
       bUseBrushTransparency=false
       roundingType=HalfHeightRadius
       width=0
       color {colorUseRule=UseColor_Specified, specifiedColor=<rgba 0 0 0 0>}
       cornerRadii=<0 0 0 0>
      }
      tintColor {colorUseRule=UseColor_Specified, specifiedColor=<rgba 1 1 1 1>}
      uVRegion {bIsValid=false, max=<0 0>, min=<0 0>}
     }
    }
   }
   slot {class=VerticalBoxSlot, properties {padding=<ltrb 0 0 0 24>}}
  }
  Text_Footnote {
   widgetClassPath=&CommonRichTextBlock
   bIsVariable=true
   parent=VerticalBox_0
   widget=Text_Footnote
   properties {
    autoWrapText=true
    bOverrideDefaultStyle=true
    text="Footnote (used for disclaimers)"
    wrapTextAt=1000
    defaultTextStyleOverride {
     overflowPolicy=Clip
     transformPolicy=None
     colorAndOpacity {colorUseRule=UseColor_Specified, specifiedColor=<rgba 1 1 1 1>}
     font {
      bForceMonospaced=false
      bMaterialIsStencil=false
      fontMaterial=null
      fontObject=&HeadingNow
      letterSpacing=0
      monospacedWidth=1
      size=18
      skewAmount=0
      typefaceFontName=74 Regular
      outlineSettings {
       bApplyOutlineToDropShadows=false
       bMiteredCorners=false
       bSeparateFillAlpha=false
       outlineMaterial=null
       outlineSize=0
       outlineColor=<rgba 0 0 0 1>
      }
     }
     highlightColor {colorUseRule=UseColor_Specified, specifiedColor=<rgba 0 0 0 1>}
     highlightShape {
      bIsDynamicallyLoaded=false
      drawAs=Image
      imageType=NoImage
      mirroring=NoMirror
      resourceName=None
      resourceObject=null
      tiling=NoTile
      imageSize=<32 32>
      margin=<ltrb 0 0 0 0>
      outlineSettings {
       bUseBrushTransparency=false
       roundingType=HalfHeightRadius
       width=0
       color {colorUseRule=UseColor_Specified, specifiedColor=<rgba 0 0 0 0>}
       cornerRadii=<0 0 0 0>
      }
      tintColor {colorUseRule=UseColor_Specified, specifiedColor=<rgba 1 1 1 1>}
      uVRegion {bIsValid=false, max=<0 0>, min=<0 0>}
     }
     selectedBackgroundColor {colorUseRule=UseColor_Foreground, specifiedColor=<rgba 1 0 1 1>}
     shadowColorAndOpacity=<rgba 0 0 0 1>
     shadowOffset=<0 0>
     strikeBrush {
      bIsDynamicallyLoaded=false
      drawAs=NoDrawType
      imageType=NoImage
      mirroring=NoMirror
      resourceName=None
      resourceObject=null
      tiling=NoTile
      imageSize=<32 32>
      margin=<ltrb 0 0 0 0>
      outlineSettings {
       bUseBrushTransparency=false
       roundingType=HalfHeightRadius
       width=0
       color {colorUseRule=UseColor_Specified, specifiedColor=<rgba 0 0 0 0>}
       cornerRadii=<0 0 0 0>
      }
      tintColor {colorUseRule=UseColor_Specified, specifiedColor=<rgba 1 1 1 1>}
      uVRegion {bIsValid=false, max=<0 0>, min=<0 0>}
     }
     underlineBrush {
      bIsDynamicallyLoaded=false
      drawAs=Image
      imageType=NoImage
      mirroring=NoMirror
      resourceName=None
      resourceObject=null
      tiling=NoTile
      imageSize=<32 32>
      margin=<ltrb 0 0 0 0>
      outlineSettings {
       bUseBrushTransparency=false
       roundingType=HalfHeightRadius
       width=0
       color {colorUseRule=UseColor_Specified, specifiedColor=<rgba 0 0 0 0>}
       cornerRadii=<0 0 0 0>
      }
      tintColor {colorUseRule=UseColor_Specified, specifiedColor=<rgba 1 1 1 1>}
      uVRegion {bIsValid=false, max=<0 0>, min=<0 0>}
     }
    }
   }
   slot {class=VerticalBoxSlot, properties {padding=<ltrb 0 0 0 24>}}
  }
  QRImage {
   widgetClassPath=&WBP_UIKit_Button_Generic
   bIsVariable=true
   parent=HorizontalBox_0
   widget=QRImage
   properties {bIsFocusable=false, backgroundClass=null, textClass=null, useBackgroundBlock=false, useTextBlock=false, iconSize=<256 256>}
   slot {class=HorizontalBoxSlot, properties {horizontalAlignment=HAlign_Right, verticalAlignment=VAlign_Top, padding=<ltrb 48 0 0 0>}}
  }
  Border_0 {widgetClassPath=&Border, namedSlotHost=QRImage, widget=Border_0, properties {padding=<ltrb 8 8 8 8>}}
  SizeBox_0 {
   widgetClassPath=&SizeBox
   parent=Border_0
   widget=SizeBox_0
   properties {bOverride_HeightOverride=true, bOverride_WidthOverride=true, heightOverride=300, widthOverride=300}
   slot {class=BorderSlot, properties {padding=<ltrb 8 8 8 8>}}
  }
  QRCodeImage {
   widgetClassPath=/QRCodeGenerator/Widgets/WBP_QRCodeGenerator
   bIsVariable=true
   parent=SizeBox_0
   widget=QRCodeImage
   properties {bOverride_Cursor=true, foregroundColor {colorUseRule=UseColor_Foreground, specifiedColor=<rgba 1 1 1 1>}}
   slot {class=SizeBoxSlot}
  }
 ]
}
