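# Submit step of the player feedback modal, covering report confirmation, sending, received and failed views.
# OnUpdateSubmitState switches views from the C++ parent's state, choosing spinner or progress-bar sending via ShouldUseProgressBarScreen.
# Block description shows only when block is selected and IsClamOneEnabled; custom down navigation from that button skips hidden toggles.
# Block and Community Rules clicks both call ToggleButton, sharing one IsToggled flag.
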WidgetBlueprint /FortniteCoreUI/UI/Foundation/Modals/Feedback/PlayerFeedback/PlayerFeedbackSubscreen_Submit
graphs [
 EventGraph {nodes [
   n1 = "On Clicked (Button_BlockUser)"!() -> (OutputDelegate, then, Button:Common Button Base Object Reference) {is=title K2Node_ComponentBoundEvent}
   > ToggleButton!(execute, self, Button:Common Button Base Object Reference=@n1.Button) -> (then)
   n2 = "On Clicked (Button_CommunityRules)"!() -> (OutputDelegate, then, Button:Common Button Base Object Reference) {is=title K2Node_ComponentBoundEvent}
   > ToggleButton(Button=@n2.Button)
   n3 = "On Selected Changed (Button_BlockUser)"!() -> (OutputDelegate, then, Button:Common Button Base Object Reference, Selected:bool) {is=title K2Node_ComponentBoundEvent}
   n4 = GetIsClamOneEnabled!(self:Fort Runtime Options Object Reference=GetRuntimeOptions()) -> (ReturnValue:bool)
   > SetVisibility(
    self=Get BlockDescriptionText() -> (BlockDescriptionText:Common Text Object Reference)
    InVisibility=Select(Option 0=Collapsed, Option 1=SelfHitTestInvisible, Index=AND Boolean(A=@n3.Selected, B=@n4))
   )
   n5 = Event UpdateProgressBar!() -> (OutputDelegate, then, PercentComplete:double) {is=title K2Node_Event}
   > n6 = Cast To WBP_UIKit_ProgressBar!(execute, Object=Get SubmitProgressBar() -> (SubmitProgressBar:Widget Object Reference)) -> (then, CastFailed, AsWBP UIKit Progress Bar:WBP UIKit Progress Bar Object Reference) {is=title K2Node_DynamicCast}
   > SetProgress!(execute, self:WBP UIKit Progress Bar Base Object Reference=@n6, Progress:double=@n5.PercentComplete) -> (then)
   n7 = Event OnUpdateSubmitState!() -> (OutputDelegate, then, NewState:EPlayerFeedbackSubmitState Enum=Start) {is=title K2Node_Event}
   n8 = ShouldUseProgressBarScreen!(self:Fort Player Feedback Modal Subscreen Submit Screen Object Reference) -> (ReturnValue:bool)
   n9 = Get SendingWithProgress() -> (SendingWithProgress:Overlay Object Reference)
   n10 = Get Sending() -> (Sending:Overlay Object Reference)
   > SetActiveWidget(
    self:Widget Switcher Object Reference=Get Switcher() -> (Switcher:Widget Switcher Object Reference)
    Widget=Select!(
     Start:Widget Object Reference=Get Confirmation() -> (Confirmation:Size Box Object Reference)
     Submitting:Widget Object Reference=Select!(Option 0:Overlay Object Reference=@n10, Option 1:Overlay Object Reference=@n9, Index:bool=@n8) -> (ReturnValue:Overlay Object Reference)
     SubmitFailed:Widget Object Reference=Get Report_NOT_Sent() -> (Report_NOT_Sent:Vertical Box Object Reference)
     SubmitSucceeded:Widget Object Reference=Get ReportReceived() -> (ReportReceived:Vertical Box Object Reference)
     Index:EPlayerFeedbackSubmitState Enum=@n7.NewState
    ) -> (ReturnValue:Widget Object Reference)
   )
  ]}
 Leaving_Button_BlockUser {nodes [
   n1 = Leaving_Button_BlockUser!() -> (then, Navigation:EUINavigation Enum) {is=title K2Node_FunctionEntry}
   > n2 = Branch(Condition="Equal (Enum)"(A:EUINavigation Enum=@n1, B:EUINavigation Enum=Down))
   n3 = Get Button_UnfollowCreator() -> (Button_UnfollowCreator:Common Button Base Object Reference)
   > n4 = Branch(Condition=IsVisible(self=@n3))
   > Return Node!(execute, ReturnValue:Widget Object Reference=@n3)
   n5 = Get Button_SendArchivedLogs() -> (Button_SendArchivedLogs:Common Button Base Object Reference)
   n6 = Branch(execute=@n4.else, Condition=IsVisible(self=@n5))
   > Return Node(ReturnValue=@n5)
   Return Node(execute=@n6.else, ReturnValue=Get Button_Submit() -> (Button_Submit:Common Button Base Object Reference))
   Return Node(execute=@n2.else)
  ]}
 ToggleButton {nodes [
   n1 = ToggleButton!() -> (then, Button:Common Button Base Object Reference) {is=title K2Node_FunctionEntry}
   n2 = Select(Option 0:bool=true, Option 1:bool=false, Index=Get IsToggled() -> (IsToggled:bool)) -> (ReturnValue:bool)
   > SetIsSelected!(execute, self:Common Button Base Object Reference=@n1, InSelected:bool=@n2, bGiveClickFeedback:bool=true) -> (then)
   > Set IsToggled(IsToggled:bool=@n2) -> (Output_Get:bool)
  ]}
]
variables [IsToggled {type=boolean}]
widgets {
 info {
  inheritedWidgetCount=18
  parentClass=/Script/FortniteUI.FortPlayerFeedbackModalSubscreen_SubmitScreen
  rootWidgetClass=&WidgetSwitcher
  widgetCount=36
 }
 widgets [
  Switcher {widgetClassPath=&WidgetSwitcher, bIsVariable=true, widget=Switcher}
  Confirmation {widgetClassPath=&SizeBox, bIsVariable=true, parent=Switcher, widget=Confirmation, slot {class=WidgetSwitcherSlot}}
  VerticalBox {widgetClassPath=&VerticalBox, parent=Confirmation, widget=VerticalBox, slot {class=SizeBoxSlot}}
  Text_Description {
   widgetClassPath=&CommonRichTextBlock
   bInherited=true
   bIsVariable=true
   parent=VerticalBox
   widget=Text_Description
   properties {
    autoWrapText=true
    bOverrideDefaultStyle=true
    mobileTextBlockScale=1.3
    text="You are about to report {PlayerName}."
    defaultTextStyleOverride {
     overflowPolicy=Clip
     transformPolicy=None
     colorAndOpacity {colorUseRule=UseColor_Specified, specifiedColor=<rgba 1 1 1 1>}
     font {
      bForceMonospaced=false
      bMaterialIsStencil=false
      fontMaterial=null
      fontObject=&HeadingNow
      letterSpacing=0
      monospacedWidth=1
      size=18
      skewAmount=0
      typefaceFontName=74 Regular
      outlineSettings {
       bApplyOutlineToDropShadows=false
       bMiteredCorners=false
       bSeparateFillAlpha=false
       outlineMaterial=null
       outlineSize=0
       outlineColor=<rgba 0 0 0 1>
      }
     }
     highlightColor {colorUseRule=UseColor_Specified, specifiedColor=<rgba 0 0 0 1>}
     highlightShape {
      bIsDynamicallyLoaded=false
      drawAs=Image
      imageType=NoImage
      mirroring=NoMirror
      resourceName=None
      resourceObject=null
      tiling=NoTile
      imageSize=<32 32>
      margin=<ltrb 0 0 0 0>
      outlineSettings {
       bUseBrushTransparency=false
       roundingType=HalfHeightRadius
       width=0
       color {colorUseRule=UseColor_Specified, specifiedColor=<rgba 0 0 0 0>}
       cornerRadii=<0 0 0 0>
      }
      tintColor {colorUseRule=UseColor_Specified, specifiedColor=<rgba 1 1 1 1>}
      uVRegion {bIsValid=false, max=<0 0>, min=<0 0>}
     }
     selectedBackgroundColor {colorUseRule=UseColor_Foreground, specifiedColor=<rgba 1 0 1 1>}
     shadowColorAndOpacity=<rgba 0 0 0 1>
     shadowOffset=<0 0>
     strikeBrush {
      bIsDynamicallyLoaded=false
      drawAs=Image
      imageType=NoImage
      mirroring=NoMirror
      resourceName=None
      resourceObject=null
      tiling=NoTile
      imageSize=<32 32>
      margin=<ltrb 0 0 0 0>
      outlineSettings {
       bUseBrushTransparency=false
       roundingType=HalfHeightRadius
       width=0
       color {colorUseRule=UseColor_Specified, specifiedColor=<rgba 0 0 0 0>}
       cornerRadii=<0 0 0 0>
      }
      tintColor {colorUseRule=UseColor_Specified, specifiedColor=<rgba 1 1 1 1>}
      uVRegion {bIsValid=false, max=<0 0>, min=<0 0>}
     }
     underlineBrush {
      bIsDynamicallyLoaded=false
      drawAs=Image
      imageType=NoImage
      mirroring=NoMirror
      resourceName=None
      resourceObject=null
      tiling=NoTile
      imageSize=<32 32>
      margin=<ltrb 0 0 0 0>
      outlineSettings {
       bUseBrushTransparency=false
       roundingType=HalfHeightRadius
       width=0
       color {colorUseRule=UseColor_Specified, specifiedColor=<rgba 0 0 0 0>}
       cornerRadii=<0 0 0 0>
      }
      tintColor {colorUseRule=UseColor_Specified, specifiedColor=<rgba 1 1 1 1>}
      uVRegion {bIsValid=false, max=<0 0>, min=<0 0>}
     }
    }
   }
   slot {class=VerticalBoxSlot, properties {padding=<ltrb 0 0 0 32>}}
  }
  Text_ReportReasonAdditionalDetails {
   widgetClassPath=&CommonRichTextBlock
   bInherited=true
   bIsVariable=true
   parent=VerticalBox
   widget=Text_ReportReasonAdditionalDetails
   properties {
    autoWrapText=true
    bAutoCollapseWithEmptyText=true
    bOverrideDefaultStyle=true
    mobileTextBlockScale=1.3
    defaultTextStyleOverride {
     overflowPolicy=Clip
     transformPolicy=None
     colorAndOpacity {colorUseRule=UseColor_Specified, specifiedColor=<rgba 1 1 1 1>}
     font {
      bForceMonospaced=false
      bMaterialIsStencil=false
      fontMaterial=null
      fontObject=&HeadingNow
      letterSpacing=0
      monospacedWidth=1
      size=18
      skewAmount=0
      typefaceFontName=74 Regular
      outlineSettings {
       bApplyOutlineToDropShadows=false
       bMiteredCorners=false
       bSeparateFillAlpha=false
       outlineMaterial=null
       outlineSize=0
       outlineColor=<rgba 0 0 0 1>
      }
     }
     highlightColor {colorUseRule=UseColor_Specified, specifiedColor=<rgba 0 0 0 1>}
     highlightShape {
      bIsDynamicallyLoaded=false
      drawAs=Image
      imageType=NoImage
      mirroring=NoMirror
      resourceName=None
      resourceObject=null
      tiling=NoTile
      imageSize=<32 32>
      margin=<ltrb 0 0 0 0>
      outlineSettings {
       bUseBrushTransparency=false
       roundingType=HalfHeightRadius
       width=0
       color {colorUseRule=UseColor_Specified, specifiedColor=<rgba 0 0 0 0>}
       cornerRadii=<0 0 0 0>
      }
      tintColor {colorUseRule=UseColor_Specified, specifiedColor=<rgba 1 1 1 1>}
      uVRegion {bIsValid=false, max=<0 0>, min=<0 0>}
     }
     selectedBackgroundColor {colorUseRule=UseColor_Foreground, specifiedColor=<rgba 1 0 1 1>}
     shadowColorAndOpacity=<rgba 0 0 0 1>
     shadowOffset=<0 0>
     strikeBrush {
      bIsDynamicallyLoaded=false
      drawAs=Image
      imageType=NoImage
      mirroring=NoMirror
      resourceName=None
      resourceObject=null
      tiling=NoTile
      imageSize=<32 32>
      margin=<ltrb 0 0 0 0>
      outlineSettings {
       bUseBrushTransparency=false
       roundingType=HalfHeightRadius
       width=0
       color {colorUseRule=UseColor_Specified, specifiedColor=<rgba 0 0 0 0>}
       cornerRadii=<0 0 0 0>
      }
      tintColor {colorUseRule=UseColor_Specified, specifiedColor=<rgba 1 1 1 1>}
      uVRegion {bIsValid=false, max=<0 0>, min=<0 0>}
     }
     underlineBrush {
      bIsDynamicallyLoaded=false
      drawAs=Image
      imageType=NoImage
      mirroring=NoMirror
      resourceName=None
      resourceObject=null
      tiling=NoTile
      imageSize=<32 32>
      margin=<ltrb 0 0 0 0>
      outlineSettings {
       bUseBrushTransparency=false
       roundingType=HalfHeightRadius
       width=0
       color {colorUseRule=UseColor_Specified, specifiedColor=<rgba 0 0 0 0>}
       cornerRadii=<0 0 0 0>
      }
      tintColor {colorUseRule=UseColor_Specified, specifiedColor=<rgba 1 1 1 1>}
      uVRegion {bIsValid=false, max=<0 0>, min=<0 0>}
     }
    }
   }
   slot {class=VerticalBoxSlot, properties {padding=<ltrb 0 0 0 32>}}
  }
  GridPanel_0 {
   widgetClassPath=&GridPanel
   parent=VerticalBox
   widget=GridPanel_0
   properties {bOverride_Cursor=true, columnFill [0,1]}
   slot {class=VerticalBoxSlot, properties {horizontalAlignment=HAlign_Left, verticalAlignment=VAlign_Center}}
  }
  Button_BlockUser {
   widgetClassPath=/FortUILibrary/UIKit/Buttons/WBP_UIKit_ButtonWithToggle
   bInherited=true
   bIsVariable=true
   parent=GridPanel_0
   widget=Button_BlockUser
   properties {
    contentHorizontalAlignment=HAlign_Right
    mobileFontSizeMultiplierOverride=1.16
    mobileMinHeight=88
    navigation=&self:WidgetTree.Button_BlockUser.WidgetNavigation_0
    text=Do you want to block this person?
    textBlockAlignment=Left
    textJustification=Left
    textSecondJustification=Left
    textTransformPolicy=None
    toggleType=Toggle
    visibility=Visible
    buttonPadding=<ltrb 0 0 0 0>
    font {
     bForceMonospaced=false
     bMaterialIsStencil=false
     fontMaterial=null
     fontObject=&HeadingNow
     letterSpacing=48
     monospacedWidth=1
     size=18
     skewAmount=0
     typefaceFontName=74 Regular
     outlineSettings {
      bApplyOutlineToDropShadows=false
      bMiteredCorners=false
      bSeparateFillAlpha=false
      outlineMaterial=null
      outlineSize=0
      outlineColor=<rgba 0 0 0 1>
     }
    }
    textPadding=<ltrb 0 12 16 12>
   }
   slot {class=GridSlot, properties {horizontalAlignment=HAlign_Left, padding=<ltrb 0 0 0 32>}}
  }
  BlockDescriptionText {
   widgetClassPath=&CommonTextBlock
   bInherited=true
   bIsVariable=true
   parent=GridPanel_0
   widget=BlockDescriptionText
   properties {
    autoWrapText=true
    bIsScrollingEnabled=false
    mobileFontSizeMultiplier=1.3
    text="Blocked players will be removed from your friends list, cannot invite you to their party, and cannot voice chat or text chat with you. Blocked players will not be notified that you blocked them."
    visibility=Collapsed
    colorAndOpacity {colorUseRule=UseColor_Specified, specifiedColor=<rgba 0.521 0.521 0.521 1>}
    font {
     bForceMonospaced=false
     bMaterialIsStencil=false
     fontMaterial=null
     fontObject=&HeadingNow
     letterSpacing=0
     monospacedWidth=1
     size=15.75
     skewAmount=0
     typefaceFontName=74 Regular
     outlineSettings {
      bApplyOutlineToDropShadows=false
      bMiteredCorners=false
      bSeparateFillAlpha=false
      outlineMaterial=null
      outlineSize=0
      outlineColor=<rgba 0 0 0 1>
     }
    }
    shadowOffset=<0 0>
   }
   slot {class=GridSlot, properties {columnSpan=2, row=1, verticalAlignment=VAlign_Top, padding=<ltrb 0 -32 0 32>}}
  }
  Button_UnfollowCreator {
   widgetClassPath=&WBP_UIKit_ButtonWithToggle
   bInherited=true
   bIsVariable=true
   parent=GridPanel_0
   widget=Button_UnfollowCreator
   properties {
    contentHorizontalAlignment=HAlign_Right
    mobileFontSizeMultiplierOverride=1.16
    mobileMinHeight=88
    navigation=&self:WidgetTree.Button_UnfollowCreator.WidgetNavigation_0
    text=Do you want to unfollow this person?
    textBlockAlignment=Left
    textJustification=Left
    textSecondJustification=Left
    textTransformPolicy=None
    toggleType=Toggle
    visibility=Visible
    buttonPadding=<ltrb 0 0 0 0>
    font {
     bForceMonospaced=false
     bMaterialIsStencil=false
     fontMaterial=null
     fontObject=&HeadingNow
     letterSpacing=48
     monospacedWidth=1
     size=18
     skewAmount=0
     typefaceFontName=74 Regular
     outlineSettings {
      bApplyOutlineToDropShadows=false
      bMiteredCorners=false
      bSeparateFillAlpha=false
      outlineMaterial=null
      outlineSize=0
      outlineColor=<rgba 0 0 0 1>
     }
    }
    textPadding=<ltrb 0 12 16 12>
   }
   slot {class=GridSlot, properties {horizontalAlignment=HAlign_Left, row=2, padding=<ltrb 0 0 0 32>}}
  }
  GridPanel_58 {
   widgetClassPath=&GridPanel
   parent=VerticalBox
   widget=GridPanel_58
   properties {columnFill [0,1]}
   slot {class=VerticalBoxSlot, properties {horizontalAlignment=HAlign_Left, verticalAlignment=VAlign_Center}}
  }
  Button_SendArchivedLogs {
   widgetClassPath=&WBP_UIKit_ButtonWithToggle
   bInherited=true
   bIsVariable=true
   parent=GridPanel_58
   widget=Button_SendArchivedLogs
   properties {
    contentHorizontalAlignment=HAlign_Right
    mobileFontSizeMultiplierOverride=1.16
    mobileMinHeight=88
    navigation=&self:WidgetTree.Button_SendArchivedLogs.WidgetNavigation_0
    text=Include Archived Logs
    textBlockAlignment=Left
    textJustification=Left
    textSecondJustification=Left
    textTransformPolicy=None
    toggleType=Toggle
    visibility=Collapsed
    buttonPadding=<ltrb 0 0 0 0>
    font {
     bForceMonospaced=false
     bMaterialIsStencil=false
     fontMaterial=null
     fontObject=&HeadingNow
     letterSpacing=48
     monospacedWidth=1
     size=18
     skewAmount=0
     typefaceFontName=74 Regular
     outlineSettings {
      bApplyOutlineToDropShadows=false
      bMiteredCorners=false
      bSeparateFillAlpha=false
      outlineMaterial=null
      outlineSize=0
      outlineColor=<rgba 0 0 0 1>
     }
    }
    textPadding=<ltrb 0 12 16 12>
   }
   slot {class=GridSlot, properties {horizontalAlignment=HAlign_Left, padding=<ltrb 0 0 0 32>}}
  }
  SendArchivedLogsText {
   widgetClassPath=&CommonTextBlock
   bInherited=true
   bIsVariable=true
   parent=GridPanel_58
   widget=SendArchivedLogsText
   properties {
    autoWrapText=true
    bIsScrollingEnabled=false
    text="Selecting this option will also include log files from previous game sessions with your report, which can help us diagnose persistent issues. This may result in increased data usage and longer report submission times."
    visibility=Collapsed
    colorAndOpacity {colorUseRule=UseColor_Specified, specifiedColor=<rgba 0.521 0.521 0.521 1>}
    font {
     bForceMonospaced=false
     bMaterialIsStencil=false
     fontMaterial=null
     fontObject=&HeadingNow
     letterSpacing=0
     monospacedWidth=1
     size=15.75
     skewAmount=0
     typefaceFontName=74 Regular
     outlineSettings {
      bApplyOutlineToDropShadows=false
      bMiteredCorners=false
      bSeparateFillAlpha=false
      outlineMaterial=null
      outlineSize=0
      outlineColor=<rgba 0 0 0 1>
     }
    }
    shadowOffset=<0 0>
   }
   slot {class=GridSlot, properties {columnSpan=2, row=1, verticalAlignment=VAlign_Top, padding=<ltrb 0 -32 0 32>}}
  }
  HorizontalBox_0 {
   widgetClassPath=&HorizontalBox
   parent=VerticalBox
   widget=HorizontalBox_0
   slot {class=VerticalBoxSlot, properties {horizontalAlignment=HAlign_Left, verticalAlignment=VAlign_Top, padding=<ltrb 0 0 0 16>}}
  }
  Button_Submit {
   widgetClassPath=/FortUILibrary/UIKit/Buttons/WBP_UIKit_Button_Loud
   bInherited=true
   bIsVariable=true
   parent=HorizontalBox_0
   widget=Button_Submit
   properties {mobileMinHeight=88, text=accept, triggeringInputAction {dataTable=&InputActionDataTable, rowName=GenericFaceButtonLeft}}
   slot {class=HorizontalBoxSlot}
  }
  Button_CommunityRules {
   widgetClassPath=&WBP_UIKit_Button_Regular
   bInherited=true
   bIsVariable=true
   parent=HorizontalBox_0
   widget=Button_CommunityRules
   properties {
    mobileMinHeight=88
    text=Community Rules
    triggeringInputAction {dataTable=&InputActionDataTable, rowName=GenericFaceButtonTop}
   }
   slot {class=HorizontalBoxSlot, properties {padding=<ltrb 32 0 0 0>}}
  }
  Sending {widgetClassPath=&Overlay, bIsVariable=true, parent=Switcher, widget=Sending, slot {class=WidgetSwitcherSlot}}
  VerticalBox_1 {
   widgetClassPath=&VerticalBox
   parent=Sending
   widget=VerticalBox_1
   slot {class=OverlaySlot, properties {horizontalAlignment=HAlign_Center, padding=<ltrb 0 140 0 0>}}
  }
  SubmitSpinner {
   widgetClassPath=/FortUILibrary/UIKit/Throbber/WBP_UIKit_Throbber
   bInherited=true
   bIsVariable=true
   parent=VerticalBox_1
   widget=SubmitSpinner
   properties {displayBackground=false, throbberSize=Medium, padding=<ltrb 0 0 0 24>}
   slot {class=VerticalBoxSlot, properties {horizontalAlignment=HAlign_Center}}
  }
  CommonTextBlock_1 {
   widgetClassPath=&CommonTextBlock
   parent=VerticalBox_1
   widget=CommonTextBlock_1
   properties {
    justification=Center
    text=Sending Report...
    font {
     bForceMonospaced=false
     bMaterialIsStencil=false
     fontMaterial=null
     fontObject=&HeadingNow
     letterSpacing=0
     monospacedWidth=1
     size=20.25
     skewAmount=0
     typefaceFontName=75 Medium
     outlineSettings {
      bApplyOutlineToDropShadows=false
      bMiteredCorners=false
      bSeparateFillAlpha=false
      outlineMaterial=null
      outlineSize=0
      outlineColor=<rgba 0 0 0 1>
     }
    }
    shadowOffset=<0 0>
   }
   slot {class=VerticalBoxSlot}
  }
  SendingWithProgress {widgetClassPath=&Overlay, bIsVariable=true, parent=Switcher, widget=SendingWithProgress, slot {class=WidgetSwitcherSlot}}
  VerticalBox_82 {
   widgetClassPath=&VerticalBox
   parent=SendingWithProgress
   widget=VerticalBox_82
   slot {class=OverlaySlot, properties {horizontalAlignment=HAlign_Fill, verticalAlignment=VAlign_Fill, padding=<ltrb 0 100 0 100>}}
  }
  CommonTextBlock {
   widgetClassPath=&CommonTextBlock
   parent=VerticalBox_82
   widget=CommonTextBlock
   properties {
    justification=Center
    text=Sending Report...
    font {
     bForceMonospaced=false
     bMaterialIsStencil=false
     fontMaterial=null
     fontObject=&HeadingNow
     letterSpacing=0
     monospacedWidth=1
     size=20.25
     skewAmount=0
     typefaceFontName=75 Medium
     outlineSettings {
      bApplyOutlineToDropShadows=false
      bMiteredCorners=false
      bSeparateFillAlpha=false
      outlineMaterial=null
      outlineSize=0
      outlineColor=<rgba 0 0 0 1>
     }
    }
    shadowOffset=<0 0>
   }
   slot {class=VerticalBoxSlot, properties {padding=<ltrb 0 0 0 10>}}
  }
  SubmitFilenameText {
   widgetClassPath=&CommonTextBlock
   bInherited=true
   bIsVariable=true
   parent=VerticalBox_82
   widget=SubmitFilenameText
   properties {
    bIsScrollingEnabled=false
    justification=Center
    colorAndOpacity {colorUseRule=UseColor_Specified, specifiedColor=<rgba 0.4983 0.4983 0.4983 1>}
    font {
     bForceMonospaced=false
     bMaterialIsStencil=false
     fontMaterial=null
     fontObject=&HeadingNow
     letterSpacing=0
     monospacedWidth=1
     size=10.5
     skewAmount=0
     typefaceFontName=75 Medium
     outlineSettings {
      bApplyOutlineToDropShadows=false
      bMiteredCorners=false
      bSeparateFillAlpha=false
      outlineMaterial=null
      outlineSize=0
      outlineColor=<rgba 0 0 0 1>
     }
    }
    shadowOffset=<0 0>
   }
   slot {class=VerticalBoxSlot, properties {horizontalAlignment=HAlign_Center, padding=<ltrb 5 0 5 0>, size {sizeRule=Fill, value=1}}}
  }
  SubmitProgressBar {
   widgetClassPath=/FortUILibrary/UIKit/ProgressBar/WBP_UIKit_ProgressBar
   bInherited=true
   bIsVariable=true
   parent=VerticalBox_82
   widget=SubmitProgressBar
   properties {height=30, progressValue=0.4, useCircular=false, width=500}
   slot {class=VerticalBoxSlot, properties {padding=<ltrb 0 20 0 5>}}
  }
  HorizontalBox_148 {
   widgetClassPath=&HorizontalBox
   parent=VerticalBox_82
   widget=HorizontalBox_148
   slot {class=VerticalBoxSlot, properties {padding=<ltrb 5 0 5 0>}}
  }
  SubmitNumericProgressText {
   widgetClassPath=&CommonTextBlock
   bInherited=true
   bIsVariable=true
   parent=HorizontalBox_148
   widget=SubmitNumericProgressText
   properties {
    bIsScrollingEnabled=false
    justification=Center
    text="0.00 B / 0.00 B [0.00 B/s]"
    colorAndOpacity {colorUseRule=UseColor_Specified, specifiedColor=<rgba 0.4983 0.4983 0.4983 1>}
    font {
     bForceMonospaced=false
     bMaterialIsStencil=false
     fontMaterial=null
     fontObject=&HeadingNow
     letterSpacing=0
     monospacedWidth=1
     size=10.5
     skewAmount=0
     typefaceFontName=75 Medium
     outlineSettings {
      bApplyOutlineToDropShadows=false
      bMiteredCorners=false
      bSeparateFillAlpha=false
      outlineMaterial=null
      outlineSize=0
      outlineColor=<rgba 0 0 0 1>
     }
    }
    shadowOffset=<0 0>
   }
   slot {class=HorizontalBoxSlot, properties {horizontalAlignment=HAlign_Left, size {sizeRule=Fill, value=1}}}
  }
  SubmitPercentProgressText {
   widgetClassPath=&CommonTextBlock
   bInherited=true
   bIsVariable=true
   parent=HorizontalBox_148
   widget=SubmitPercentProgressText
   properties {
    bIsScrollingEnabled=false
    justification=Center
    text=0.00%
    colorAndOpacity {colorUseRule=UseColor_Specified, specifiedColor=<rgba 0.4983 0.4983 0.4983 1>}
    font {
     bForceMonospaced=false
     bMaterialIsStencil=false
     fontMaterial=null
     fontObject=&HeadingNow
     letterSpacing=0
     monospacedWidth=1
     size=10.5
     skewAmount=0
     typefaceFontName=75 Medium
     outlineSettings {
      bApplyOutlineToDropShadows=false
      bMiteredCorners=false
      bSeparateFillAlpha=false
      outlineMaterial=null
      outlineSize=0
      outlineColor=<rgba 0 0 0 1>
     }
    }
    shadowOffset=<0 0>
   }
   slot {class=HorizontalBoxSlot, properties {horizontalAlignment=HAlign_Right}}
  }
  ReportReceived {widgetClassPath=&VerticalBox, bIsVariable=true, parent=Switcher, widget=ReportReceived, slot {class=WidgetSwitcherSlot}}
  Text_Summary {
   widgetClassPath=&CommonTextBlock
   bInherited=true
   parent=ReportReceived
   widget=Text_Summary
   properties {
    autoWrapText=true
    mobileFontSizeMultiplier=1.3
    text="Report received; we'll investigate."
    wrapTextAt=800
    font {
     bForceMonospaced=false
     bMaterialIsStencil=false
     fontMaterial=null
     fontObject=&HeadingNow
     letterSpacing=0
     monospacedWidth=1
     size=18
     skewAmount=0
     typefaceFontName=74 Regular
     outlineSettings {
      bApplyOutlineToDropShadows=false
      bMiteredCorners=false
      bSeparateFillAlpha=false
      outlineMaterial=null
      outlineSize=0
      outlineColor=<rgba 0 0 0 1>
     }
    }
    shadowOffset=<0 0>
   }
   slot {class=VerticalBoxSlot, properties {padding=<ltrb 0 0 0 16>}}
  }
  Text_ReportNumber {
   widgetClassPath=/Script/FortniteUI.FortRichTextBlockLegacy
   bInherited=true
   bIsVariable=true
   parent=ReportReceived
   widget=Text_ReportNumber
   properties {
    autoWrapText=true
    styleSet=/Game/UI/Foundation/Modals/Feedback/PlayerFeedback/FeedbackScreenRichTextStyles
    text="Report Number ##xxxxxxxx"
    wrapTextAt=800
    hyperlinkButtonStyle {
     clickedSlateSound {resourceObject=null}
     disabled {
      bIsDynamicallyLoaded=false
      drawAs=NoDrawType
      imageType=NoImage
      mirroring=NoMirror
      resourceName=None
      resourceObject=null
      tiling=NoTile
      imageSize=<0 0>
      margin=<ltrb 0 0 0 0>
      outlineSettings {
       bUseBrushTransparency=false
       roundingType=HalfHeightRadius
       width=0
       color {colorUseRule=UseColor_Specified, specifiedColor=<rgba 0 0 0 0>}
       cornerRadii=<0 0 0 0>
      }
      tintColor {colorUseRule=UseColor_Specified, specifiedColor=<rgba 1 1 1 1>}
      uVRegion {bIsValid=false, max=<0 0>, min=<0 0>}
     }
     disabledForeground {colorUseRule=UseColor_Foreground, specifiedColor=<rgba 1 0 1 1>}
     hovered {
      bIsDynamicallyLoaded=false
      drawAs=Image
      imageType=NoImage
      mirroring=NoMirror
      resourceName=None
      resourceObject=null
      tiling=NoTile
      imageSize=<32 32>
      margin=<ltrb 0 0 0 0>
      outlineSettings {
       bUseBrushTransparency=false
       roundingType=HalfHeightRadius
       width=0
       color {colorUseRule=UseColor_Specified, specifiedColor=<rgba 0 0 0 0>}
       cornerRadii=<0 0 0 0>
      }
      tintColor {colorUseRule=UseColor_Specified, specifiedColor=<rgba 1 1 1 1>}
      uVRegion {bIsValid=false, max=<0 0>, min=<0 0>}
     }
     hoveredForeground {colorUseRule=UseColor_Foreground, specifiedColor=<rgba 1 0 1 1>}
     hoveredSlateSound {resourceObject=null}
     normal {
      bIsDynamicallyLoaded=false
      drawAs=Image
      imageType=NoImage
      mirroring=NoMirror
      resourceName=None
      resourceObject=null
      tiling=NoTile
      imageSize=<32 32>
      margin=<ltrb 0 0 0 0>
      outlineSettings {
       bUseBrushTransparency=false
       roundingType=HalfHeightRadius
       width=0
       color {colorUseRule=UseColor_Specified, specifiedColor=<rgba 0 0 0 0>}
       cornerRadii=<0 0 0 0>
      }
      tintColor {colorUseRule=UseColor_Specified, specifiedColor=<rgba 1 1 1 0>}
      uVRegion {bIsValid=false, max=<0 0>, min=<0 0>}
     }
     normalForeground {colorUseRule=UseColor_Foreground, specifiedColor=<rgba 1 0 1 1>}
     normalPadding=<ltrb 0 0 0 0>
     pressed {
      bIsDynamicallyLoaded=false
      drawAs=Image
      imageType=NoImage
      mirroring=NoMirror
      resourceName=None
      resourceObject=null
      tiling=NoTile
      imageSize=<32 32>
      margin=<ltrb 0 0 0 0>
      outlineSettings {
       bUseBrushTransparency=false
       roundingType=HalfHeightRadius
       width=0
       color {colorUseRule=UseColor_Specified, specifiedColor=<rgba 0 0 0 0>}
       cornerRadii=<0 0 0 0>
      }
      tintColor {colorUseRule=UseColor_Specified, specifiedColor=<rgba 1 1 1 1>}
      uVRegion {bIsValid=false, max=<0 0>, min=<0 0>}
     }
     pressedForeground {colorUseRule=UseColor_Foreground, specifiedColor=<rgba 1 0 1 1>}
     pressedPadding=<ltrb 0 0 0 0>
     pressedSlateSound {resourceObject=null}
    }
   }
   slot {class=VerticalBoxSlot, properties {padding=<ltrb 0 0 0 16>}}
  }
  CommonTextBlock_4 {
   widgetClassPath=&CommonTextBlock
   parent=ReportReceived
   widget=CommonTextBlock_4
   properties {
    autoWrapText=true
    mobileFontSizeMultiplier=1.3
    text="Reports help us to ensure that Fortnite remains a fair, safe and enjoyable game for all players."
    font {
     bForceMonospaced=false
     bMaterialIsStencil=false
     fontMaterial=null
     fontObject=&HeadingNow
     letterSpacing=0
     monospacedWidth=1
     size=18
     skewAmount=0
     typefaceFontName=74 Regular
     outlineSettings {
      bApplyOutlineToDropShadows=false
      bMiteredCorners=false
      bSeparateFillAlpha=false
      outlineMaterial=null
      outlineSize=0
      outlineColor=<rgba 0 0 0 1>
     }
    }
    shadowOffset=<0 0>
   }
   slot {class=VerticalBoxSlot, properties {padding=<ltrb 0 0 0 24>}}
  }
  Report_NOT_Sent {widgetClassPath=&VerticalBox, bIsVariable=true, parent=Switcher, widget=Report_NOT_Sent, slot {class=WidgetSwitcherSlot}}
  CommonTextBlock_6 {
   widgetClassPath=&CommonTextBlock
   parent=Report_NOT_Sent
   widget=CommonTextBlock_6
   properties {
    text="OOOPS!"
    font {
     bForceMonospaced=false
     bMaterialIsStencil=false
     fontMaterial=null
     fontObject=&HeadingNow
     letterSpacing=0
     monospacedWidth=1
     size=20.25
     skewAmount=0
     typefaceFontName=75 Medium
     outlineSettings {
      bApplyOutlineToDropShadows=false
      bMiteredCorners=false
      bSeparateFillAlpha=false
      outlineMaterial=null
      outlineSize=0
      outlineColor=<rgba 0 0 0 1>
     }
    }
    shadowOffset=<0 0>
   }
   slot {class=VerticalBoxSlot, properties {padding=<ltrb 0 0 0 20>}}
  }
  CommonTextBlock_7 {
   widgetClassPath=&CommonTextBlock
   parent=Report_NOT_Sent
   widget=CommonTextBlock_7
   properties {
    autoWrapText=true
    mobileFontSizeMultiplier=1.3
    text=Report failed to send.
    wrapTextAt=800
    font {
     bForceMonospaced=false
     bMaterialIsStencil=false
     fontMaterial=null
     fontObject=&HeadingNow
     letterSpacing=0
     monospacedWidth=1
     size=18
     skewAmount=0
     typefaceFontName=74 Regular
     outlineSettings {
      bApplyOutlineToDropShadows=false
      bMiteredCorners=false
      bSeparateFillAlpha=false
      outlineMaterial=null
      outlineSize=0
      outlineColor=<rgba 0 0 0 1>
     }
    }
    shadowOffset=<0 0>
   }
   slot {class=VerticalBoxSlot, properties {padding=<ltrb 0 0 0 16>}}
  }
  Text_FailReason {
   widgetClassPath=&CommonTextBlock
   bInherited=true
   parent=Report_NOT_Sent
   widget=Text_FailReason
   properties {
    mobileFontSizeMultiplier=1.3
    text=Error Message:
    wrapTextAt=800
    font {
     bForceMonospaced=false
     bMaterialIsStencil=false
     fontMaterial=null
     fontObject=&HeadingNow
     letterSpacing=0
     monospacedWidth=1
     size=18
     skewAmount=0
     typefaceFontName=75 Medium
     outlineSettings {
      bApplyOutlineToDropShadows=false
      bMiteredCorners=false
      bSeparateFillAlpha=false
      outlineMaterial=null
      outlineSize=0
      outlineColor=<rgba 0 0 0 1>
     }
    }
    shadowOffset=<0 0>
   }
   slot {class=VerticalBoxSlot}
  }
  Slot_CommunityRulesURL {widgetClassPath=/Script/UMG.Widget, bInherited=true, widget=null}
 ]
}
