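# Player feedback modal subscreen for submitting a voice chat report on another player, built on FortPlayerFeedbackModalSubscreen_SubmitVoiceScreen.
# A widget switcher follows the submit state, showing confirmation, sending, success or failure panels.
# The confirmation panel has a block-user checkbox toggle, a Community Rules button and Submit, which SetLoadingDisplay disables while showing a voice audio loading spinner.
# On construct, icon sizes and padding adjust for mobile.
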WidgetBlueprint /FortniteCoreUI/UI/Foundation/Modals/Feedback/PlayerFeedback/PlayerFeedbackSubscreen_SubmitVoiceReport
graphs [EventGraph {nodes [
   n1 = "On Selected Changed (Button_BlockUser)"!() -> (OutputDelegate, then, Button:Common Button Legacy Object Reference, Selected:bool) {is=title K2Node_ComponentBoundEvent}
   > SetVisibility(
    self=Get Image_Check_Block_1() -> (Image_Check_Block_1:Image Object Reference)
    InVisibility=Select(Option 0=Collapsed, Option 1=Visible, Index=@n1.Selected)
   )
   n2 = Event SetLoadingDisplay!() -> (OutputDelegate, then, bIsDisplayed:bool) {is=title K2Node_Event}
   > Set Enabled And Tooltip Text!(
    execute
    Widget:Widget Object Reference=Get Button_Submit() -> (Button_Submit:Common Button Base Object Reference)
    Enabled:bool=NOT Boolean(A=@n2.bIsDisplayed)
    Tooltip Text:Text
   ) -> (then) {is=title K2Node_CallFunction}
   > SetVisibility(
    self=Get HBLoadingState() -> (HBLoadingState:Horizontal Box Object Reference)
    InVisibility=Select(Option 0=Collapsed, Option 1=SelfHitTestInvisible, Index=@n2.bIsDisplayed)
   )
   Event Construct()
   n3 = IsMobileGame()
   > SetDesiredSizeOverride!(
    execute
    self:Image Object Reference=Get ProgressSpinner() -> (ProgressSpinner:Image Object Reference)
    DesiredSize:Vector 2D Structure=Select(Option 0:Vector 2D Structure=(X=28,Y=28), Option 1:Vector 2D Structure=(X=42,Y=42), Index=@n3) -> (ReturnValue:Vector 2D Structure)
   ) -> (then)
   > SetDesiredSizeOverride(
    self=Get Step_Header_1() -> (Step_Header_1:Image Object Reference)
    DesiredSize=Select(Option 0=(X=32,Y=32), Option 1=(X=48,Y=48), Index=@n3)
   )
   > SetPadding!(
    execute
    self:Horizontal Box Slot Object Reference=SlotAsHorizontalBoxSlot!(Widget:Widget Object Reference=Get Step_Header_1()) -> (ReturnValue:Horizontal Box Slot Object Reference)
    InPadding_Left:float=0
    InPadding_Top:float=0
    InPadding_Right:float=Select(Option 0:float=0, Option 1:float=4, Index=@n3) -> (ReturnValue:float)
    InPadding_Bottom:float=0
   ) -> (then)
   > SetPadding(
    self=SlotAsHorizontalBoxSlot(Widget=Get ProgressSpinner())
    InPadding_Left=0
    InPadding_Top=0
    InPadding_Right=Select(Option 0=14, Option 1=16, Index=@n3)
    InPadding_Bottom=0
   )
   n4 = Event OnUpdateSubmitState!() -> (OutputDelegate, then, NewState:EPlayerFeedbackSubmitState Enum=Start) {is=title K2Node_Event}
   > SetActiveWidget(
    self:Widget Switcher Object Reference=Get Switcher() -> (Switcher:Widget Switcher Object Reference)
    Widget=Select!(
     Start:Widget Object Reference=Get Confirmation() -> (Confirmation:Size Box Object Reference)
     Submitting:Widget Object Reference=Get Sending() -> (Sending:Vertical Box Object Reference)
     SubmitFailed:Widget Object Reference=Get Report_NOT_Sent() -> (Report_NOT_Sent:Vertical Box Object Reference)
     SubmitSucceeded:Widget Object Reference=Get ReportReceived() -> (ReportReceived:Vertical Box Object Reference)
     Index:EPlayerFeedbackSubmitState Enum=@n4.NewState
    ) -> (ReturnValue:Widget Object Reference)
   )
  ]}]
variables []
widgets {
 info {
  inheritedWidgetCount=18
  namedSlotCount=2
  parentClass=/Script/FortniteUI.FortPlayerFeedbackModalSubscreen_SubmitVoiceScreen
  rootWidgetClass=&WidgetSwitcher
  widgetCount=45
 }
 widgets [
  Switcher {widgetClassPath=&WidgetSwitcher, bIsVariable=true, widget=Switcher}
  Confirmation {widgetClassPath=&SizeBox, bIsVariable=true, parent=Switcher, widget=Confirmation, slot {class=WidgetSwitcherSlot}}
  VerticalBox {widgetClassPath=&VerticalBox, parent=Confirmation, widget=VerticalBox, slot {class=SizeBoxSlot}}
  Text_Description {
   widgetClassPath=&CommonRichTextBlock
   bInherited=true
   bIsVariable=true
   parent=VerticalBox
   widget=Text_Description
   properties {
    autoWrapText=true
    bOverrideDefaultStyle=true
    justification=Center
    text="You are about to report {PlayerName}."
    defaultTextStyleOverride {
     overflowPolicy=Clip
     transformPolicy=None
     colorAndOpacity {colorUseRule=UseColor_Specified, specifiedColor=<rgba 0.3185 0.7231 1 1>}
     font {
      bForceMonospaced=false
      bMaterialIsStencil=false
      fontMaterial=null
      fontObject=&BurbankSmall
      letterSpacing=0
      monospacedWidth=1
      size=26
      skewAmount=0
      typefaceFontName=Bold
      outlineSettings {
       bApplyOutlineToDropShadows=false
       bMiteredCorners=false
       bSeparateFillAlpha=false
       outlineMaterial=null
       outlineSize=0
       outlineColor=<rgba 0 0 0 1>
      }
     }
     highlightColor {colorUseRule=UseColor_Specified, specifiedColor=<rgba 0 0 0 1>}
     highlightShape {
      bIsDynamicallyLoaded=false
      drawAs=Image
      imageType=NoImage
      mirroring=NoMirror
      resourceName=None
      resourceObject=null
      tiling=NoTile
      imageSize=<32 32>
      margin=<ltrb 0 0 0 0>
      outlineSettings {
       bUseBrushTransparency=false
       roundingType=HalfHeightRadius
       width=0
       color {colorUseRule=UseColor_Specified, specifiedColor=<rgba 0 0 0 0>}
       cornerRadii=<0 0 0 0>
      }
      tintColor {colorUseRule=UseColor_Specified, specifiedColor=<rgba 1 1 1 1>}
      uVRegion {bIsValid=false, max=<0 0>, min=<0 0>}
     }
     selectedBackgroundColor {colorUseRule=UseColor_Foreground, specifiedColor=<rgba 1 0 1 1>}
     shadowColorAndOpacity=<rgba 0 0 0 1>
     shadowOffset=<0 0>
     strikeBrush {
      bIsDynamicallyLoaded=false
      drawAs=Image
      imageType=NoImage
      mirroring=NoMirror
      resourceName=None
      resourceObject=null
      tiling=NoTile
      imageSize=<32 32>
      margin=<ltrb 0 0 0 0>
      outlineSettings {
       bUseBrushTransparency=false
       roundingType=HalfHeightRadius
       width=0
       color {colorUseRule=UseColor_Specified, specifiedColor=<rgba 0 0 0 0>}
       cornerRadii=<0 0 0 0>
      }
      tintColor {colorUseRule=UseColor_Specified, specifiedColor=<rgba 1 1 1 1>}
      uVRegion {bIsValid=false, max=<0 0>, min=<0 0>}
     }
     underlineBrush {
      bIsDynamicallyLoaded=false
      drawAs=Image
      imageType=NoImage
      mirroring=NoMirror
      resourceName=None
      resourceObject=null
      tiling=NoTile
      imageSize=<32 32>
      margin=<ltrb 0 0 0 0>
      outlineSettings {
       bUseBrushTransparency=false
       roundingType=HalfHeightRadius
       width=0
       color {colorUseRule=UseColor_Specified, specifiedColor=<rgba 0 0 0 0>}
       cornerRadii=<0 0 0 0>
      }
      tintColor {colorUseRule=UseColor_Specified, specifiedColor=<rgba 1 1 1 1>}
      uVRegion {bIsValid=false, max=<0 0>, min=<0 0>}
     }
    }
   }
   slot {class=VerticalBoxSlot, properties {horizontalAlignment=HAlign_Center, padding=<ltrb 20 40 20 30>}}
  }
  VerticalBox_0 {
   widgetClassPath=&VerticalBox
   parent=VerticalBox
   widget=VerticalBox_0
   slot {class=VerticalBoxSlot, properties {horizontalAlignment=HAlign_Center, verticalAlignment=VAlign_Center, padding=<ltrb 0 0 0 20>}}
  }
  Button_BlockUser {
   widgetClassPath=&IconTextButton
   bInherited=true
   bIsVariable=true
   parent=VerticalBox_0
   widget=Button_BlockUser
   properties {
    bInteractableWhenSelected=true
    bSelectable=true
    bToggleable=true
    buttonText=Do you want to block this person?
    buttonTextJustification=Left
    navigation=&self:WidgetTree.Button_BlockUser.WidgetNavigation_0
    onButtonBaseClicked=(&self:WidgetTree.Button_BlockUser.HandleOnButtonClicked)
    onButtonBaseDoubleClicked=(&self:WidgetTree.Button_BlockUser.HandleOnButtonDoubleClicked)
    onButtonBaseHovered=(&self:WidgetTree.Button_BlockUser.HandleOnButtonHovered)
    onButtonBaseUnhovered=(&self:WidgetTree.Button_BlockUser.HandleOnButtonUnhovered)
    onSelectedChangedBase=(&self:WidgetTree.Button_BlockUser.HandleOnSelectedChanged)
    style=/FortniteCoreUI/UI/Foundation/Modals/PlayerReporting/ButtonStyle-PlayerReport.ButtonStyle-PlayerReport_C
    visibility=Visible
   }
   slot {class=VerticalBoxSlot}
  }
  SizeBox_1 {
   widgetClassPath=&SizeBox
   namedSlotHost=Button_BlockUser
   widget=SizeBox_1
   properties {bOverride_HeightOverride=true, bOverride_WidthOverride=true, heightOverride=56, widthOverride=56}
  }
  DoYouKnowCheck_4 {widgetClassPath=&Overlay, parent=SizeBox_1, widget=DoYouKnowCheck_4, slot {class=SizeBoxSlot}}
  CheckBox1_1 {
   widgetClassPath=&Image
   bIsVariable=true
   parent=DoYouKnowCheck_4
   widget=CheckBox1_1
   properties {brush {
     bIsDynamicallyLoaded=false
     drawAs=Image
     imageType=NoImage
     mirroring=NoMirror
     resourceName=None
     resourceObject=/FortniteCoreUI/Athena/UI/Challenges/Art/MissionTileImages/M_UI_CheckBox
     tiling=NoTile
     imageSize=<64 64>
     margin=<ltrb 0 0 0 0>
     outlineSettings {
      bUseBrushTransparency=false
      roundingType=HalfHeightRadius
      width=0
      color {colorUseRule=UseColor_Specified, specifiedColor=<rgba 0 0 0 0>}
      cornerRadii=<0 0 0 0>
     }
     tintColor {colorUseRule=UseColor_Specified, specifiedColor=<rgba 1 1 1 1>}
     uVRegion {bIsValid=false, max=<0 0>, min=<0 0>}
    }}
   slot {class=OverlaySlot, properties {horizontalAlignment=HAlign_Center, verticalAlignment=VAlign_Bottom}}
  }
  Image_Check_Block_1 {
   widgetClassPath=&Image
   bIsVariable=true
   parent=DoYouKnowCheck_4
   widget=Image_Check_Block_1
   properties {
    visibility=Collapsed
    brush {
     bIsDynamicallyLoaded=false
     drawAs=Image
     imageType=NoImage
     mirroring=NoMirror
     resourceName=None
     resourceObject=/Game/UI/Foundation/Textures/Icons/Checks/T-Icon-Check-64
     tiling=NoTile
     imageSize=<54 54>
     margin=<ltrb 0 0 0 0>
     outlineSettings {
      bUseBrushTransparency=false
      roundingType=HalfHeightRadius
      width=0
      color {colorUseRule=UseColor_Specified, specifiedColor=<rgba 0 0 0 0>}
      cornerRadii=<0 0 0 0>
     }
     tintColor {colorUseRule=UseColor_Specified, specifiedColor=<rgba 0.0343 1 0.227 1>}
     uVRegion {bIsValid=false, max=<0 0>, min=<0 0>}
    }
    renderTransform {angle=0, scale=<1 1>, shear=<0 0>, translation=<6 -9>}
   }
   slot {class=OverlaySlot, properties {horizontalAlignment=HAlign_Center, verticalAlignment=VAlign_Center}}
  }
  Slot_VoiceChatAudio {
   widgetClassPath=&VerticalBox
   bIsVariable=true
   parent=VerticalBox
   widget=Slot_VoiceChatAudio
   slot {class=VerticalBoxSlot, properties {horizontalAlignment=HAlign_Center, verticalAlignment=VAlign_Top, padding=<ltrb 20 0 20 20>}}
  }
  HorizontalBox_262 {
   widgetClassPath=&HorizontalBox
   parent=Slot_VoiceChatAudio
   widget=HorizontalBox_262
   slot {class=VerticalBoxSlot, properties {horizontalAlignment=HAlign_Left, verticalAlignment=VAlign_Center}}
  }
  Step_Header_1 {
   widgetClassPath=&Image
   bIsVariable=true
   parent=HorizontalBox_262
   widget=Step_Header_1
   properties {brush {
     bIsDynamicallyLoaded=false
     drawAs=Image
     imageType=NoImage
     mirroring=NoMirror
     resourceName=None
     resourceObject=/VoiceRep/Materials/MI_UI_Icon_Paperclip
     tiling=NoTile
     imageSize=<32 32>
     margin=<ltrb 0 0 0 0>
     outlineSettings {
      bUseBrushTransparency=false
      roundingType=HalfHeightRadius
      width=0
      color {colorUseRule=UseColor_Specified, specifiedColor=<rgba 0 0 0 0>}
      cornerRadii=<0 0 0 0>
     }
     tintColor {colorUseRule=UseColor_Specified, specifiedColor=<rgba 1 1 1 1>}
     uVRegion {bIsValid=false, max=<0 0>, min=<0 0>}
    }}
   slot {class=HorizontalBoxSlot, properties {horizontalAlignment=HAlign_Left, verticalAlignment=VAlign_Center, padding=<ltrb 8 0 0 0>}}
  }
  Text_Label_2 {
   widgetClassPath=&CommonTextBlock
   parent=HorizontalBox_262
   widget=Text_Label_2
   properties {
    bIsScrollingEnabled=false
    mobileFontSizeMultiplier=1.5
    text=Voice Chat audio up to 5 minutes
    visibility=Visible
    wrapTextAt=600
    colorAndOpacity {colorUseRule=UseColor_Specified, specifiedColor=<rgba 0.0319 0.7011 1 1>}
    font {
     bForceMonospaced=false
     bMaterialIsStencil=false
     fontMaterial=null
     fontObject=&BurbankSmall
     letterSpacing=0
     monospacedWidth=1
     size=15
     skewAmount=0
     typefaceFontName=Bold
     outlineSettings {
      bApplyOutlineToDropShadows=false
      bMiteredCorners=false
      bSeparateFillAlpha=false
      outlineMaterial=null
      outlineSize=0
      outlineColor=<rgba 0 0 0 1>
     }
    }
    shadowOffset=<0 0>
   }
   slot {class=HorizontalBoxSlot, properties {horizontalAlignment=HAlign_Left, verticalAlignment=VAlign_Center, padding=<ltrb 2 0 16 0>}}
  }
  Spacer_114 {widgetClassPath=&Spacer, parent=HorizontalBox_262, widget=Spacer_114, properties {size=<0 45>}, slot {class=HorizontalBoxSlot}}
  HBLoadingState {
   widgetClassPath=&HorizontalBox
   bIsVariable=true
   parent=Slot_VoiceChatAudio
   widget=HBLoadingState
   properties {visibility=Collapsed}
   slot {class=VerticalBoxSlot, properties {horizontalAlignment=HAlign_Center, padding=<ltrb 0 50 0 0>}}
  }
  ProgressSpinner {
   widgetClassPath=&Image
   bIsVariable=true
   parent=HBLoadingState
   widget=ProgressSpinner
   properties {
    brush {
     bIsDynamicallyLoaded=false
     drawAs=Image
     imageType=NoImage
     mirroring=NoMirror
     resourceName=None
     resourceObject=/Game/UI/Foundation/Materials/Throbber_Material-Monochrome
     tiling=NoTile
     imageSize=<28 28>
     margin=<ltrb 0 0 0 0>
     outlineSettings {
      bUseBrushTransparency=false
      roundingType=HalfHeightRadius
      width=0
      color {colorUseRule=UseColor_Specified, specifiedColor=<rgba 0 0 0 0>}
      cornerRadii=<0 0 0 0>
     }
     tintColor {colorUseRule=UseColor_Specified, specifiedColor=<rgba 1 1 1 1>}
     uVRegion {bIsValid=false, max=<0 0>, min=<0 0>}
    }
    colorAndOpacity=<rgba 0.0331 0.7011 1 1>
   }
   slot {class=HorizontalBoxSlot, properties {horizontalAlignment=HAlign_Center, verticalAlignment=VAlign_Center, padding=<ltrb 12 0 14 0>}}
  }
  Text_Loading_State {
   widgetClassPath=&CommonTextBlock
   bIsVariable=true
   parent=HBLoadingState
   widget=Text_Loading_State
   properties {
    bIsScrollingEnabled=false
    mobileFontSizeMultiplier=1.5
    text=Loading voice chat audio...
    visibility=Visible
    wrapTextAt=600
    colorAndOpacity {colorUseRule=UseColor_Specified, specifiedColor=<rgba 0.0331 0.7011 1 1>}
    font {
     bForceMonospaced=false
     bMaterialIsStencil=false
     fontMaterial=null
     fontObject=&BurbankSmall
     letterSpacing=0
     monospacedWidth=1
     size=15
     skewAmount=0
     typefaceFontName=Bold
     outlineSettings {
      bApplyOutlineToDropShadows=false
      bMiteredCorners=false
      bSeparateFillAlpha=false
      outlineMaterial=null
      outlineSize=0
      outlineColor=<rgba 0 0 0 1>
     }
    }
    shadowOffset=<0 0>
   }
   slot {class=HorizontalBoxSlot, properties {horizontalAlignment=HAlign_Left, verticalAlignment=VAlign_Center, padding=<ltrb 0 0 16 0>}}
  }
  Button_CommunityRules {
   widgetClassPath=&IconTextButton
   bInherited=true
   bIsVariable=true
   parent=VerticalBox
   widget=Button_CommunityRules
   properties {
    buttonText=Community Rules
    contentAlignment=HAlign_Center
    minHeight=64
    minWidth=550
    onButtonBaseClicked=(&self:WidgetTree.Button_CommunityRules.HandleOnButtonClicked)
    onButtonBaseDoubleClicked=(&self:WidgetTree.Button_CommunityRules.HandleOnButtonDoubleClicked)
    onButtonBaseHovered=(&self:WidgetTree.Button_CommunityRules.HandleOnButtonHovered)
    onButtonBaseUnhovered=(&self:WidgetTree.Button_CommunityRules.HandleOnButtonUnhovered)
    onSelectedChangedBase=(&self:WidgetTree.Button_CommunityRules.HandleOnSelectedChanged)
    style=/FortUILibrary/Foundation/Buttons/ButtonStyles/Tab/ButtonStyle-Tab-Main-Recolor.ButtonStyle-Tab-Main-Recolor_C
    iconBrush {
     bIsDynamicallyLoaded=false
     drawAs=Image
     imageType=NoImage
     mirroring=NoMirror
     resourceName=None
     resourceObject=/FortniteCoreUI/UI/Login/Art/M-Icon-ExternalLink-64
     tiling=NoTile
     imageSize=<0 0>
     margin=<ltrb 0 0 0 0>
     outlineSettings {
      bUseBrushTransparency=false
      roundingType=HalfHeightRadius
      width=0
      color {colorUseRule=UseColor_Specified, specifiedColor=<rgba 0 0 0 0>}
      cornerRadii=<0 0 0 1>
     }
     tintColor {colorUseRule=UseColor_Specified, specifiedColor=<rgba 1 1 1 1>}
     uVRegion {bIsValid=false, max=<0 0>, min=<0 0>}
    }
    triggeringInputAction {dataTable=&InputActionDataTable, rowName=GenericFaceButtonTop}
   }
   slot {class=VerticalBoxSlot, properties {horizontalAlignment=HAlign_Center, verticalAlignment=VAlign_Top, padding=<ltrb 0 0 0 20>}}
  }
  Image_1 {
   widgetClassPath=&Image
   bIsVariable=true
   namedSlotHost=Button_CommunityRules
   widget=Image_1
   properties {
    brush {
     bIsDynamicallyLoaded=false
     drawAs=Image
     imageType=NoImage
     mirroring=NoMirror
     resourceName=None
     resourceObject=&M-Icon-ExternalLink-64
     tiling=NoTile
     imageSize=<64 70>
     margin=<ltrb 0 0 0 0>
     outlineSettings {
      bUseBrushTransparency=false
      roundingType=HalfHeightRadius
      width=0
      color {colorUseRule=UseColor_Specified, specifiedColor=<rgba 0 0 0 0>}
      cornerRadii=<0 0 0 0>
     }
     tintColor {colorUseRule=UseColor_Specified, specifiedColor=<rgba 0.3185 0.7231 1 1>}
     uVRegion {bIsValid=false, max=<0 0>, min=<0 0>}
    }
    renderTransform {angle=0, scale=<1 1>, shear=<0 0>, translation=<0 -2>}
   }
  }
  Button_Submit {
   widgetClassPath=&IconTextButton
   bInherited=true
   bIsVariable=true
   parent=VerticalBox
   widget=Button_Submit
   properties {
    bIsEnabled=false
    buttonText=submit
    contentAlignment=HAlign_Center
    minHeight=64
    minWidth=550
    onButtonBaseClicked=(&self:WidgetTree.Button_Submit.HandleOnButtonClicked)
    onButtonBaseDoubleClicked=(&self:WidgetTree.Button_Submit.HandleOnButtonDoubleClicked)
    onButtonBaseHovered=(&self:WidgetTree.Button_Submit.HandleOnButtonHovered)
    onButtonBaseUnhovered=(&self:WidgetTree.Button_Submit.HandleOnButtonUnhovered)
    onSelectedChangedBase=(&self:WidgetTree.Button_Submit.HandleOnSelectedChanged)
    style=/FortUILibrary/Foundation/Buttons/ButtonStyles/Skew/ButtonStyle-Skew_Desirable.ButtonStyle-Skew_Desirable_C
    triggeringInputAction {dataTable=&InputActionDataTable, rowName=GenericFaceButtonLeft}
   }
   slot {class=VerticalBoxSlot, properties {horizontalAlignment=HAlign_Center, verticalAlignment=VAlign_Center}}
  }
  Sending {widgetClassPath=&VerticalBox, bIsVariable=true, parent=Switcher, widget=Sending, slot {class=WidgetSwitcherSlot}}
  CommonTextBlock_1 {
   widgetClassPath=&CommonTextBlock
   parent=Sending
   widget=CommonTextBlock_1
   properties {
    justification=Center
    text=Sending Report...
    textTransformPolicy=ToUpper
    font {
     bForceMonospaced=false
     bMaterialIsStencil=false
     fontMaterial=null
     fontObject=&BurbankSmall
     letterSpacing=0
     monospacedWidth=1
     size=24
     skewAmount=0
     typefaceFontName=Bold
     outlineSettings {
      bApplyOutlineToDropShadows=false
      bMiteredCorners=false
      bSeparateFillAlpha=false
      outlineMaterial=null
      outlineSize=0
      outlineColor=<rgba 0 0 0 1>
     }
    }
    shadowOffset=<0 0>
   }
   slot {class=VerticalBoxSlot, properties {padding=<ltrb 0 40 0 40>}}
  }
  CircularThrobber_0 {
   widgetClassPath=/Script/UMG.CircularThrobber
   bIsVariable=true
   parent=Sending
   widget=CircularThrobber_0
   properties {
    numberOfPieces=10
    period=1.5
    radius=75
    image {
     bIsDynamicallyLoaded=false
     drawAs=Image
     imageType=FullColor
     mirroring=NoMirror
     resourceName=../../../Engine/Content/Slate/Common/Throbber_Piece.png
     resourceObject=/Game/UI/Materials/M_UICircle
     tiling=NoTile
     imageSize=<20 20>
     margin=<ltrb 0 0 0 0>
     outlineSettings {
      bUseBrushTransparency=false
      roundingType=HalfHeightRadius
      width=0
      color {colorUseRule=UseColor_Specified, specifiedColor=<rgba 0 0 0 0>}
      cornerRadii=<0 0 0 0>
     }
     tintColor {colorUseRule=UseColor_Specified, specifiedColor=<rgba 0.6038 0.8879 1 1>}
     uVRegion {bIsValid=false, max=<0 0>, min=<0 0>}
    }
   }
   slot {class=VerticalBoxSlot, properties {horizontalAlignment=HAlign_Center, verticalAlignment=VAlign_Center}}
  }
  ReportReceived {widgetClassPath=&VerticalBox, bIsVariable=true, parent=Switcher, widget=ReportReceived, slot {class=WidgetSwitcherSlot}}
  CommonTextBlock_2 {
   widgetClassPath=&CommonTextBlock
   parent=ReportReceived
   widget=CommonTextBlock_2
   properties {
    text="THANKS!"
    font {
     bForceMonospaced=false
     bMaterialIsStencil=false
     fontMaterial=null
     fontObject=&BurbankSmall
     letterSpacing=0
     monospacedWidth=1
     size=50
     skewAmount=0
     typefaceFontName=Black
     outlineSettings {
      bApplyOutlineToDropShadows=false
      bMiteredCorners=false
      bSeparateFillAlpha=false
      outlineMaterial=null
      outlineSize=0
      outlineColor=<rgba 0 0 0 1>
     }
    }
    shadowOffset=<0 0>
   }
   slot {class=VerticalBoxSlot, properties {horizontalAlignment=HAlign_Center, padding=<ltrb 0 0 0 30>}}
  }
  Text_Summary {
   widgetClassPath=&CommonTextBlock
   bInherited=true
   parent=ReportReceived
   widget=Text_Summary
   properties {
    autoWrapText=true
    justification=Center
    mobileFontSizeMultiplier=1.3
    text="Report received; we'll investigate."
    wrapTextAt=800
    colorAndOpacity {colorUseRule=UseColor_Specified, specifiedColor=<rgba 0.6038 0.8879 1 1>}
    font {
     bForceMonospaced=false
     bMaterialIsStencil=false
     fontMaterial=null
     fontObject=&BurbankSmall
     letterSpacing=0
     monospacedWidth=1
     size=18
     skewAmount=0
     typefaceFontName=Bold
     outlineSettings {
      bApplyOutlineToDropShadows=false
      bMiteredCorners=false
      bSeparateFillAlpha=false
      outlineMaterial=null
      outlineSize=0
      outlineColor=<rgba 0 0 0 1>
     }
    }
    shadowOffset=<0 0>
   }
   slot {class=VerticalBoxSlot, properties {verticalAlignment=VAlign_Top, padding=<ltrb 0 0 0 20>}}
  }
  Text_ReportNumber {
   widgetClassPath=/Script/FortniteUI.FortRichTextBlockLegacy
   bInherited=true
   bIsVariable=true
   parent=ReportReceived
   widget=Text_ReportNumber
   properties {
    autoWrapText=true
    justification=Center
    styleSet=/Game/UI/Foundation/Modals/Feedback/PlayerFeedback/FeedbackVoiceRichTextStyles
    text="Report Number ##xxxxxxxx"
    wrapTextAt=800
    hyperlinkButtonStyle {
     clickedSlateSound {resourceObject=null}
     disabled {
      bIsDynamicallyLoaded=false
      drawAs=NoDrawType
      imageType=NoImage
      mirroring=NoMirror
      resourceName=None
      resourceObject=null
      tiling=NoTile
      imageSize=<0 0>
      margin=<ltrb 0 0 0 0>
      outlineSettings {
       bUseBrushTransparency=false
       roundingType=HalfHeightRadius
       width=0
       color {colorUseRule=UseColor_Specified, specifiedColor=<rgba 0 0 0 0>}
       cornerRadii=<0 0 0 0>
      }
      tintColor {colorUseRule=UseColor_Specified, specifiedColor=<rgba 1 1 1 1>}
      uVRegion {bIsValid=false, max=<0 0>, min=<0 0>}
     }
     disabledForeground {colorUseRule=UseColor_Foreground, specifiedColor=<rgba 1 0 1 1>}
     hovered {
      bIsDynamicallyLoaded=false
      drawAs=Image
      imageType=NoImage
      mirroring=NoMirror
      resourceName=None
      resourceObject=null
      tiling=NoTile
      imageSize=<32 32>
      margin=<ltrb 0 0 0 0>
      outlineSettings {
       bUseBrushTransparency=false
       roundingType=HalfHeightRadius
       width=0
       color {colorUseRule=UseColor_Specified, specifiedColor=<rgba 0 0 0 0>}
       cornerRadii=<0 0 0 0>
      }
      tintColor {colorUseRule=UseColor_Specified, specifiedColor=<rgba 1 1 1 1>}
      uVRegion {bIsValid=false, max=<0 0>, min=<0 0>}
     }
     hoveredForeground {colorUseRule=UseColor_Foreground, specifiedColor=<rgba 1 0 1 1>}
     hoveredSlateSound {resourceObject=null}
     normal {
      bIsDynamicallyLoaded=false
      drawAs=Image
      imageType=NoImage
      mirroring=NoMirror
      resourceName=None
      resourceObject=null
      tiling=NoTile
      imageSize=<32 32>
      margin=<ltrb 0 0 0 0>
      outlineSettings {
       bUseBrushTransparency=false
       roundingType=HalfHeightRadius
       width=0
       color {colorUseRule=UseColor_Specified, specifiedColor=<rgba 0 0 0 0>}
       cornerRadii=<0 0 0 0>
      }
      tintColor {colorUseRule=UseColor_Specified, specifiedColor=<rgba 1 1 1 0>}
      uVRegion {bIsValid=false, max=<0 0>, min=<0 0>}
     }
     normalForeground {colorUseRule=UseColor_Foreground, specifiedColor=<rgba 1 0 1 1>}
     normalPadding=<ltrb 0 0 0 0>
     pressed {
      bIsDynamicallyLoaded=false
      drawAs=Image
      imageType=NoImage
      mirroring=NoMirror
      resourceName=None
      resourceObject=null
      tiling=NoTile
      imageSize=<32 32>
      margin=<ltrb 0 0 0 0>
      outlineSettings {
       bUseBrushTransparency=false
       roundingType=HalfHeightRadius
       width=0
       color {colorUseRule=UseColor_Specified, specifiedColor=<rgba 0 0 0 0>}
       cornerRadii=<0 0 0 0>
      }
      tintColor {colorUseRule=UseColor_Specified, specifiedColor=<rgba 1 1 1 1>}
      uVRegion {bIsValid=false, max=<0 0>, min=<0 0>}
     }
     pressedForeground {colorUseRule=UseColor_Foreground, specifiedColor=<rgba 1 0 1 1>}
     pressedPadding=<ltrb 0 0 0 0>
     pressedSlateSound {resourceObject=null}
    }
   }
   slot {class=VerticalBoxSlot, properties {verticalAlignment=VAlign_Top, padding=<ltrb 0 0 0 20>}}
  }
  CommonTextBlock_4 {
   widgetClassPath=&CommonTextBlock
   parent=ReportReceived
   widget=CommonTextBlock_4
   properties {
    justification=Center
    mobileFontSizeMultiplier=1.3
    text="Your report helps us keep Fortnite a fair, safe, and fun game for everyone."
    wrapTextAt=800
    colorAndOpacity {colorUseRule=UseColor_Specified, specifiedColor=<rgba 0.6038 0.8879 1 1>}
    font {
     bForceMonospaced=false
     bMaterialIsStencil=false
     fontMaterial=null
     fontObject=&BurbankSmall
     letterSpacing=0
     monospacedWidth=1
     size=18
     skewAmount=0
     typefaceFontName=Bold
     outlineSettings {
      bApplyOutlineToDropShadows=false
      bMiteredCorners=false
      bSeparateFillAlpha=false
      outlineMaterial=null
      outlineSize=0
      outlineColor=<rgba 0 0 0 1>
     }
    }
    shadowOffset=<0 0>
   }
   slot {class=VerticalBoxSlot, properties {verticalAlignment=VAlign_Top, padding=<ltrb 0 0 0 20>}}
  }
  CommonTextBlock_5 {
   widgetClassPath=&CommonTextBlock
   parent=ReportReceived
   widget=CommonTextBlock_5
   properties {
    autoWrapText=true
    justification=Center
    mobileFontSizeMultiplier=1.3
    text="Good luck, have fun!"
    wrapTextAt=800
    colorAndOpacity {colorUseRule=UseColor_Specified, specifiedColor=<rgba 0.6038 0.8879 1 1>}
    font {
     bForceMonospaced=false
     bMaterialIsStencil=false
     fontMaterial=null
     fontObject=&BurbankSmall
     letterSpacing=0
     monospacedWidth=1
     size=18
     skewAmount=0
     typefaceFontName=Bold
     outlineSettings {
      bApplyOutlineToDropShadows=false
      bMiteredCorners=false
      bSeparateFillAlpha=false
      outlineMaterial=null
      outlineSize=0
      outlineColor=<rgba 0 0 0 1>
     }
    }
    shadowOffset=<0 0>
   }
   slot {class=VerticalBoxSlot, properties {verticalAlignment=VAlign_Top}}
  }
  Report_NOT_Sent {widgetClassPath=&VerticalBox, bIsVariable=true, parent=Switcher, widget=Report_NOT_Sent, slot {class=WidgetSwitcherSlot}}
  CommonTextBlock_6 {
   widgetClassPath=&CommonTextBlock
   parent=Report_NOT_Sent
   widget=CommonTextBlock_6
   properties {
    text="OOOPS!"
    font {
     bForceMonospaced=false
     bMaterialIsStencil=false
     fontMaterial=null
     fontObject=&BurbankSmall
     letterSpacing=0
     monospacedWidth=1
     size=50
     skewAmount=0
     typefaceFontName=Black
     outlineSettings {
      bApplyOutlineToDropShadows=false
      bMiteredCorners=false
      bSeparateFillAlpha=false
      outlineMaterial=null
      outlineSize=0
      outlineColor=<rgba 0 0 0 1>
     }
    }
    shadowOffset=<0 0>
   }
   slot {class=VerticalBoxSlot, properties {horizontalAlignment=HAlign_Center, padding=<ltrb 0 0 0 30>}}
  }
  CommonTextBlock_7 {
   widgetClassPath=&CommonTextBlock
   parent=Report_NOT_Sent
   widget=CommonTextBlock_7
   properties {
    autoWrapText=true
    justification=Center
    mobileFontSizeMultiplier=1.3
    text=Report failed to send.
    wrapTextAt=800
    colorAndOpacity {colorUseRule=UseColor_Specified, specifiedColor=<rgba 0.6038 0.8879 1 1>}
    font {
     bForceMonospaced=false
     bMaterialIsStencil=false
     fontMaterial=null
     fontObject=&BurbankSmall
     letterSpacing=0
     monospacedWidth=1
     size=18
     skewAmount=0
     typefaceFontName=Bold
     outlineSettings {
      bApplyOutlineToDropShadows=false
      bMiteredCorners=false
      bSeparateFillAlpha=false
      outlineMaterial=null
      outlineSize=0
      outlineColor=<rgba 0 0 0 1>
     }
    }
    shadowOffset=<0 0>
   }
   slot {class=VerticalBoxSlot, properties {verticalAlignment=VAlign_Top, padding=<ltrb 0 0 0 20>}}
  }
  Text_FailReason {
   widgetClassPath=&CommonTextBlock
   bInherited=true
   parent=Report_NOT_Sent
   widget=Text_FailReason
   properties {
    justification=Center
    mobileFontSizeMultiplier=1.3
    text=Error Message:
    wrapTextAt=800
    colorAndOpacity {colorUseRule=UseColor_Specified, specifiedColor=<rgba 0.6038 0.8879 1 1>}
    font {
     bForceMonospaced=false
     bMaterialIsStencil=false
     fontMaterial=null
     fontObject=&BurbankSmall
     letterSpacing=0
     monospacedWidth=1
     size=18
     skewAmount=0
     typefaceFontName=Bold
     outlineSettings {
      bApplyOutlineToDropShadows=false
      bMiteredCorners=false
      bSeparateFillAlpha=false
      outlineMaterial=null
      outlineSize=0
      outlineColor=<rgba 0 0 0 1>
     }
    }
    shadowOffset=<0 0>
   }
   slot {class=VerticalBoxSlot, properties {verticalAlignment=VAlign_Top}}
  }
  Slot_CommunityRulesURL {widgetClassPath=/Script/UMG.Widget, bInherited=true, widget=null}
  Text_ReportReasonAdditionalDetails {widgetClassPath=&CommonRichTextBlock, bInherited=true, widget=null}
  BlockDescriptionText {widgetClassPath=&CommonTextBlock, bInherited=true, widget=null}
  Button_UnfollowCreator {widgetClassPath=&CommonButtonBase, bInherited=true, widget=null}
  Button_SendArchivedLogs {widgetClassPath=&CommonButtonBase, bInherited=true, widget=null}
  SendArchivedLogsText {widgetClassPath=&CommonTextBlock, bInherited=true, widget=null}
  SubmitSpinner {widgetClassPath=&Widget, bInherited=true, widget=null}
  SubmitProgressBar {widgetClassPath=&Widget, bInherited=true, widget=null}
  SubmitFilenameText {widgetClassPath=&CommonTextBlock, bInherited=true, widget=null}
  SubmitNumericProgressText {widgetClassPath=&CommonTextBlock, bInherited=true, widget=null}
  SubmitPercentProgressText {widgetClassPath=&CommonTextBlock, bInherited=true, widget=null}
 ]
}
